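# Physics asset for the front apron cloth on the Angel female medium body, named as the crouch variant.
# Capsules on pelvis, thighs and calves form the leg body shapes.
# Left and right box-body chains run from pelvis through apron driver bones to four dynamic segments each.
# Most joints have free angular motion, while the 02-03 joints and cross-links tying both sides at segments 03 and 04 are more tightly constrained.
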PhysicsAsset /BRCosmetics/Characters/Player/Female/Medium/Bodies/F_MED_Angel/Meshes/F_MED_Angel_FrontCloth_Crouch_Physics
bodies {
 calf_l [{shapeType=Capsule, length=33.9939, radius=6.978, center=<-25.0142 -5.2215 0.3081>, rotation=<rot 6.9224 -82.7948 -90.4849>}]
 calf_r [{shapeType=Capsule, length=30.9471, radius=8.2553, center=<24.6819 3.4927 0.3127>, rotation=<rot 4.1437 -84.619 -92.6178>}]
 driver_apron_front_01_l [{shapeType=Box, extentX=1, extentY=1, extentZ=1, rotation=<rot 0 -90 0>}]
 driver_apron_front_01_r [{shapeType=Box, extentX=1, extentY=1, extentZ=1, rotation=<rot 0 -90 0>}]
 dyn_apron_frnt_rb_01_l [{shapeType=Box, extentX=1, extentY=1, extentZ=1}]
 dyn_apron_frnt_rb_01_r [{shapeType=Box, extentX=1.5, extentY=1, extentZ=2}]
 dyn_apron_frnt_rb_02_l [{shapeType=Box, extentX=10, extentY=1.5, extentZ=12, center=<3.5 0 0>}]
 dyn_apron_frnt_rb_02_r [{shapeType=Box, extentX=9.5409, extentY=2.9925, extentZ=12.0878, center=<-3.1192 -0.0001 0>}]
 dyn_apron_frnt_rb_03_l [{shapeType=Box, extentX=1, extentY=9.4188, extentZ=11.4536, center=<5 0.0001 0>, rotation=<rot 0 90 90>}]
 dyn_apron_frnt_rb_03_r [{shapeType=Box, extentX=0.963, extentY=10.0083, extentZ=11.7209, center=<-5 0 0>, rotation=<rot 0 -90 -90>}]
 dyn_apron_frnt_rb_04_l [{shapeType=Box, extentX=1.5, extentY=10, extentZ=22, center=<10.6 0 0>, rotation=<rot 0 90 90>}]
 dyn_apron_frnt_rb_04_r [{shapeType=Box, extentX=1.5, extentY=10, extentZ=22, center=<-10.6 0 0>, rotation=<rot 0 90 -90>}]
 pelvis [{shapeType=Capsule, length=12.2049, radius=8.9393}]
 thigh_l [{shapeType=Capsule, length=41.7527, radius=8.8318, center=<-25.2298 0.2368 0.6489>, rotation=<rot 0.0106 -88.0555 -89.9143>}]
 thigh_r [{shapeType=Capsule, length=43.1071, radius=8.3871, center=<24.5517 -0.6274 -0.6152>, rotation=<rot -0.0843 -88.4 -90.9026>}]
}
constraints [
 {
  bone1Name=thigh_l
  bone2Name=pelvis
  swing1LimitDegrees=45
  swing1Motion=Free
  swing2LimitDegrees=45
  swing2Motion=Free
  twistLimitDegrees=45
  twistMotion=Free
 }
 {
  bone1Name=calf_l
  bone2Name=thigh_l
  swing1LimitDegrees=45
  swing1Motion=Free
  swing2LimitDegrees=45
  swing2Motion=Free
  twistLimitDegrees=45
  twistMotion=Free
 }
 {
  bone1Name=thigh_r
  bone2Name=pelvis
  swing1LimitDegrees=45
  swing1Motion=Free
  swing2LimitDegrees=45
  swing2Motion=Free
  twistLimitDegrees=45
  twistMotion=Free
 }
 {
  bone1Name=calf_r
  bone2Name=thigh_r
  swing1LimitDegrees=45
  swing1Motion=Free
  swing2LimitDegrees=45
  swing2Motion=Free
  twistLimitDegrees=45
  twistMotion=Free
 }
 {bone1Name=dyn_apron_frnt_rb_03_l, bone2Name=dyn_apron_frnt_rb_02_l, swing1LimitDegrees=45, swing2LimitDegrees=45, twistLimitDegrees=8}
 {
  bone1Name=dyn_apron_frnt_rb_04_l
  bone2Name=dyn_apron_frnt_rb_03_l
  swing1LimitDegrees=60
  swing1Motion=Free
  swing2LimitDegrees=60
  swing2Motion=Free
  twistLimitDegrees=15
  twistMotion=Free
 }
 {
  bone1Name=dyn_apron_frnt_rb_02_l
  bone2Name=dyn_apron_frnt_rb_01_l
  swing1LimitDegrees=45
  swing1Motion=Free
  swing2LimitDegrees=45
  swing2Motion=Free
  twistLimitDegrees=45
  twistMotion=Free
 }
 {
  bone1Name=dyn_apron_frnt_rb_02_r
  bone2Name=dyn_apron_frnt_rb_01_r
  swing1LimitDegrees=45
  swing1Motion=Free
  swing2LimitDegrees=45
  swing2Motion=Free
  twistLimitDegrees=45
  twistMotion=Free
 }
 {
  bone1Name=dyn_apron_frnt_rb_01_r
  bone2Name=driver_apron_front_01_r
  swing1LimitDegrees=45
  swing1Motion=Free
  swing2LimitDegrees=45
  swing2Motion=Free
  twistLimitDegrees=45
  twistMotion=Free
 }
 {
  bone1Name=driver_apron_front_01_r
  bone2Name=pelvis
  swing1LimitDegrees=45
  swing1Motion=Free
  swing2LimitDegrees=45
  swing2Motion=Free
  twistLimitDegrees=45
  twistMotion=Free
 }
 {bone1Name=dyn_apron_frnt_rb_03_r, bone2Name=dyn_apron_frnt_rb_02_r, swing1LimitDegrees=45, swing2LimitDegrees=45, twistLimitDegrees=5}
 {
  bone1Name=dyn_apron_frnt_rb_04_r
  bone2Name=dyn_apron_frnt_rb_03_r
  swing1LimitDegrees=60
  swing1Motion=Free
  swing2LimitDegrees=60
  swing2Motion=Free
  twistLimitDegrees=15
  twistMotion=Free
 }
 {bone1Name=dyn_apron_frnt_rb_03_l, bone2Name=dyn_apron_frnt_rb_03_r, swing1LimitDegrees=20, swing2LimitDegrees=20, twistLimitDegrees=5}
 {
  bone1Name=dyn_apron_frnt_rb_01_l
  bone2Name=driver_apron_front_01_l
  swing1LimitDegrees=45
  swing1Motion=Free
  swing2LimitDegrees=45
  swing2Motion=Free
  twistLimitDegrees=45
  twistMotion=Free
 }
 {
  bone1Name=driver_apron_front_01_l
  bone2Name=pelvis
  swing1LimitDegrees=45
  swing1Motion=Free
  swing2LimitDegrees=45
  swing2Motion=Free
  twistLimitDegrees=45
  twistMotion=Free
 }
 {bone1Name=dyn_apron_frnt_rb_04_r, bone2Name=dyn_apron_frnt_rb_04_l, swing1LimitDegrees=20, swing2LimitDegrees=20, twistLimitDegrees=5}
]
