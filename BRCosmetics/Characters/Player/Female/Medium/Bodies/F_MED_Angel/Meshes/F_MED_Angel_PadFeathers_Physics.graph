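# Physics asset for the left shoulder pad and feathers of the female medium Angel body cosmetic.
# Capsule bodies cover a small root pad bone, two pad pieces (A and B) and four identical feather bones.
# Both pad pieces are constrained to the root, and all four feathers to piece B, forming a shallow tree.
# Narrow swing and twist limits are set per joint, tightest on piece A.
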PhysicsAsset /BRCosmetics/Characters/Player/Female/Medium/Bodies/F_MED_Angel/Meshes/F_MED_Angel_PadFeathers_Physics
bodies {
 dyn_pad_feather_01 [{shapeType=Capsule, length=2, radius=1.25, center=<0.7 0 0>, rotation=<rot -0.0004 0 0>}]
 dyn_pad_feather_02 [{shapeType=Capsule, length=2, radius=1.25, center=<0.7 0 0>, rotation=<rot -0.0004 0 0>}]
 dyn_pad_feather_03 [{shapeType=Capsule, length=2, radius=1.25, center=<0.7 0 0>, rotation=<rot -0.0004 0 0>}]
 dyn_pad_feather_04 [{shapeType=Capsule, length=2, radius=1.25, center=<0.7 0 0>, rotation=<rot -0.0003 0 0.0001>}]
 dyn_shlderpadA_l [{shapeType=Capsule, length=6.1934, radius=2.945, center=<0.68 0 0>, rotation=<rot 90 0 0>}]
 dyn_shlderpadB_l [{shapeType=Capsule, length=3, radius=3.1202, center=<0.45 0 0>, rotation=<rot 90 0 0>}]
 dyn_shldrpad_l [{shapeType=Capsule, length=0.505, radius=0.505, rotation=<rot 0 -90 0>}]
}
constraints [
 {bone1Name=dyn_shlderpadB_l, bone2Name=dyn_shldrpad_l, swing1LimitDegrees=20, swing2LimitDegrees=20, twistLimitDegrees=8}
 {bone1Name=dyn_shlderpadA_l, bone2Name=dyn_shldrpad_l, swing1LimitDegrees=10, swing2LimitDegrees=10, twistLimitDegrees=5}
 {bone1Name=dyn_pad_feather_01, bone2Name=dyn_shlderpadB_l, swing1LimitDegrees=15, swing2LimitDegrees=15, twistLimitDegrees=5}
 {bone1Name=dyn_pad_feather_02, bone2Name=dyn_shlderpadB_l, swing1LimitDegrees=15, swing2LimitDegrees=15, twistLimitDegrees=5}
 {bone1Name=dyn_pad_feather_03, bone2Name=dyn_shlderpadB_l, swing1LimitDegrees=15, swing2LimitDegrees=15, twistLimitDegrees=5}
 {bone1Name=dyn_pad_feather_04, bone2Name=dyn_shlderpadB_l, swing1LimitDegrees=15, swing2LimitDegrees=15, twistLimitDegrees=5}
]
