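# Physics asset for the F_MED_Angel female player body, set up for a dynamic front loincloth.
# Left and right three-box chains branch from a middle cloth box constrained to the pelvis, with limited swing and twist.
# Cross constraints tie the two chains at each level, angularly free at the top and limited below.
# Thigh capsules give the legs collision volume, while pelvis and spine_02 carry only tiny capsules.
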PhysicsAsset /BRCosmetics/Characters/Player/Female/Medium/Bodies/F_MED_Angel/Meshes/F_MED_Angel_Physics
bodies {
 dyn_cloth_font_mid_01 [{shapeType=Box, extentX=14.138, extentY=18.8765, extentZ=2.8776, center=<7.4134 0 0>}]
 dyn_loincloth_01_l [{shapeType=Box, extentX=11.7301, extentY=11.6363, extentZ=3.5, center=<5.6591 3 -0.0001>}]
 dyn_loincloth_01_r [{shapeType=Box, extentX=11.7301, extentY=11.6363, extentZ=3.5, center=<-5.6591 -3 -0.0002>}]
 dyn_loincloth_02_l [{shapeType=Box, extentX=11.7301, extentY=11.6363, extentZ=3.5, center=<7.0619 2.32 0>}]
 dyn_loincloth_02_r [{shapeType=Box, extentX=11.7301, extentY=11.6363, extentZ=3.5, center=<-7.0619 -2.32 0>}]
 dyn_loincloth_03_l [{shapeType=Box, extentX=11.7301, extentY=11.6363, extentZ=3.5, center=<5.6591 1.85 0>}]
 dyn_loincloth_03_r [{shapeType=Box, extentX=11.7301, extentY=11.6363, extentZ=3.5, center=<-5.6591 -1.85 0>}]
 pelvis [{shapeType=Capsule, length=2, radius=2}]
 spine_02 [{shapeType=Capsule, length=2, radius=2}]
 thigh_l [{shapeType=Capsule, length=39.3609, radius=11.2036, center=<-21.8097 -2.6183 -1.632>, rotation=<rot 90 0 2.1763>}]
 thigh_r [{shapeType=Capsule, length=39.3609, radius=11.2036, center=<22.0455 1.6852 -0.0001>, rotation=<rot 90 -6.831 -4.6548>}]
}
constraints [
 {bone1Name=dyn_loincloth_01_l, bone2Name=dyn_cloth_font_mid_01, swing1LimitDegrees=30, swing2LimitDegrees=30, twistLimitDegrees=45}
 {bone1Name=dyn_loincloth_02_l, bone2Name=dyn_loincloth_01_l, swing1LimitDegrees=45, swing2LimitDegrees=45, twistLimitDegrees=70}
 {bone1Name=dyn_loincloth_03_l, bone2Name=dyn_loincloth_02_l, swing1LimitDegrees=45, swing2LimitDegrees=45, twistLimitDegrees=70}
 {bone1Name=dyn_loincloth_01_r, bone2Name=dyn_cloth_font_mid_01, swing1LimitDegrees=30, swing2LimitDegrees=30, twistLimitDegrees=45}
 {bone1Name=dyn_loincloth_02_r, bone2Name=dyn_loincloth_01_r, swing1LimitDegrees=45, swing2LimitDegrees=45, twistLimitDegrees=70}
 {bone1Name=dyn_loincloth_03_r, bone2Name=dyn_loincloth_02_r, swing1LimitDegrees=45, swing2LimitDegrees=45, twistLimitDegrees=70}
 {
  bone1Name=spine_02
  bone2Name=pelvis
  swing1LimitDegrees=45
  swing1Motion=Free
  swing2LimitDegrees=45
  swing2Motion=Free
  twistLimitDegrees=45
  twistMotion=Free
 }
 {bone1Name=dyn_cloth_font_mid_01, bone2Name=pelvis, swing1LimitDegrees=20, swing2LimitDegrees=70, twistLimitDegrees=20}
 {
  bone1Name=thigh_l
  bone2Name=pelvis
  swing1LimitDegrees=45
  swing1Motion=Free
  swing2LimitDegrees=45
  swing2Motion=Free
  twistLimitDegrees=45
  twistMotion=Free
 }
 {
  bone1Name=thigh_r
  bone2Name=pelvis
  swing1LimitDegrees=45
  swing1Motion=Free
  swing2LimitDegrees=45
  swing2Motion=Free
  twistLimitDegrees=45
  twistMotion=Free
 }
 {
  bone1Name=dyn_loincloth_01_r
  bone2Name=dyn_loincloth_01_l
  swing1LimitDegrees=45
  swing1Motion=Free
  swing2LimitDegrees=45
  swing2Motion=Free
  twistLimitDegrees=45
  twistMotion=Free
 }
 {bone1Name=dyn_loincloth_02_r, bone2Name=dyn_loincloth_02_l, swing1LimitDegrees=45, swing2LimitDegrees=45, twistLimitDegrees=45}
 {bone1Name=dyn_loincloth_03_r, bone2Name=dyn_loincloth_03_l, swing1LimitDegrees=45, swing2LimitDegrees=45, twistLimitDegrees=45}
]
