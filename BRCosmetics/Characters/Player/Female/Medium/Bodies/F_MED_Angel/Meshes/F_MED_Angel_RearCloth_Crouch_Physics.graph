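# Physics asset for the Angel female medium body's rear skirt, named as the crouch variant, defining bodies for bone-driven cloth motion.
# Left and right three-segment chains of flat boxes hang from small driver bodies with limited swing and tighter twist.
# Tighter cross-constraints tie the two chains together at each level, keeping the panels moving as one sheet.
# Pelvis, thigh and calf capsules provide leg shapes, joined by free angular constraints, as are drivers to pelvis.
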PhysicsAsset /BRCosmetics/Characters/Player/Female/Medium/Bodies/F_MED_Angel/Meshes/F_MED_Angel_RearCloth_Crouch_Physics
bodies {
 calf_l [{shapeType=Capsule, length=31.4939, radius=9.478, center=<-24.8921 -3.7518 0.3126>, rotation=<rot 6.9224 -82.7948 -90.4849>}]
 calf_r [{shapeType=Capsule, length=30.4471, radius=8.7553, center=<24.5597 2.023 0.3082>, rotation=<rot 4.1437 -84.619 -92.6178>}]
 driver_skirt_rear_01_l [{shapeType=Box, extentX=1.01, extentY=1.01, extentZ=1.01, rotation=<rot 0 -90 0>}]
 driver_skirt_rear_01_r [{shapeType=Box, extentX=1.01, extentY=1.01, extentZ=1.01, rotation=<rot 0 -90 0>}]
 dyn_skirt_rear_01_l [{shapeType=Box, extentX=9.01, extentY=16.51, extentZ=1.01, center=<7 2.7 0>, rotation=<rot 0 -90 0>}]
 dyn_skirt_rear_01_r [{shapeType=Box, extentX=9.01, extentY=16.51, extentZ=1.01, center=<7 -2.7 0>, rotation=<rot 0 -90 0>}]
 dyn_skirt_rear_02_l [{shapeType=Box, extentX=10.51, extentY=16.51, extentZ=1.01, center=<7 2.5 0>, rotation=<rot 0 -90 0>}]
 dyn_skirt_rear_02_r [{shapeType=Box, extentX=10.51, extentY=16.51, extentZ=1.01, center=<7 -2.5 0>, rotation=<rot 0 -90 0>}]
 dyn_skirt_rear_03_l [{shapeType=Box, extentX=12.591, extentY=16.2649, extentZ=1.5, center=<7 3.5 0>, rotation=<rot 0 90 0>}]
 dyn_skirt_rear_03_r [{shapeType=Box, extentX=12.7723, extentY=16.7372, extentZ=2.1976, center=<7 -3.5 0>, rotation=<rot 0 90 0>}]
 pelvis [{shapeType=Capsule, length=12.2049, radius=8.9393}]
 thigh_l [{shapeType=Capsule, length=18.0027, radius=8.3318, center=<-28.5006 -0.0316 0.6768>, rotation=<rot 0 -81.0105 -89.9141>}]
 thigh_r [{shapeType=Capsule, length=18.3571, radius=8.6371, center=<29.4134 0.477 -0.5367>, rotation=<rot 0 -83.0664 -90.9065>}]
}
constraints [
 {bone1Name=dyn_skirt_rear_01_r, bone2Name=driver_skirt_rear_01_r, swing1LimitDegrees=45, swing2LimitDegrees=45, twistLimitDegrees=10}
 {bone1Name=dyn_skirt_rear_01_l, bone2Name=driver_skirt_rear_01_l, swing1LimitDegrees=45, swing2LimitDegrees=45, twistLimitDegrees=10}
 {bone1Name=dyn_skirt_rear_01_r, bone2Name=dyn_skirt_rear_01_l, swing1LimitDegrees=15, swing2LimitDegrees=15, twistLimitDegrees=15}
 {bone1Name=dyn_skirt_rear_02_r, bone2Name=dyn_skirt_rear_01_r, swing1LimitDegrees=45, swing2LimitDegrees=45, twistLimitDegrees=5}
 {bone1Name=dyn_skirt_rear_02_l, bone2Name=dyn_skirt_rear_01_l, swing1LimitDegrees=45, swing2LimitDegrees=45, twistLimitDegrees=5}
 {bone1Name=dyn_skirt_rear_02_l, bone2Name=dyn_skirt_rear_02_r, swing1LimitDegrees=15, swing2LimitDegrees=15, twistLimitDegrees=15}
 {bone1Name=dyn_skirt_rear_03_r, bone2Name=dyn_skirt_rear_02_r, swing1LimitDegrees=45, swing2LimitDegrees=45, twistLimitDegrees=5}
 {bone1Name=dyn_skirt_rear_03_l, bone2Name=dyn_skirt_rear_02_l, swing1LimitDegrees=45, swing2LimitDegrees=45, twistLimitDegrees=5}
 {bone1Name=dyn_skirt_rear_03_r, bone2Name=dyn_skirt_rear_03_l, swing1LimitDegrees=15, swing2LimitDegrees=15, twistLimitDegrees=15}
 {
  bone1Name=driver_skirt_rear_01_l
  bone2Name=pelvis
  swing1LimitDegrees=45
  swing1Motion=Free
  swing2LimitDegrees=45
  swing2Motion=Free
  twistLimitDegrees=45
  twistMotion=Free
 }
 {
  bone1Name=driver_skirt_rear_01_r
  bone2Name=pelvis
  swing1LimitDegrees=45
  swing1Motion=Free
  swing2LimitDegrees=45
  swing2Motion=Free
  twistLimitDegrees=45
  twistMotion=Free
 }
 {
  bone1Name=thigh_l
  bone2Name=pelvis
  swing1LimitDegrees=45
  swing1Motion=Free
  swing2LimitDegrees=45
  swing2Motion=Free
  twistLimitDegrees=45
  twistMotion=Free
 }
 {
  bone1Name=calf_l
  bone2Name=thigh_l
  swing1LimitDegrees=45
  swing1Motion=Free
  swing2LimitDegrees=45
  swing2Motion=Free
  twistLimitDegrees=45
  twistMotion=Free
 }
 {
  bone1Name=thigh_r
  bone2Name=pelvis
  swing1LimitDegrees=45
  swing1Motion=Free
  swing2LimitDegrees=45
  swing2Motion=Free
  twistLimitDegrees=45
  twistMotion=Free
 }
 {
  bone1Name=calf_r
  bone2Name=thigh_r
  swing1LimitDegrees=45
  swing1Motion=Free
  swing2LimitDegrees=45
  swing2Motion=Free
  twistLimitDegrees=45
  twistMotion=Free
 }
]
