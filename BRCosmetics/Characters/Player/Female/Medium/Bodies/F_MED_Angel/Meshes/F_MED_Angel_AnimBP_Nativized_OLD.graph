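# Animation Blueprint for the F_MED_Angel character body, named as an old nativized version, that adds secondary motion to skirt, cloth and feather bones.
# The pose copied from the attached parent mesh runs through AnimDynamics for skirt and calf feathers, then alpha-driven bone offsets for crouch, skydiving and other states, into a cached pose.
# RigidBody passes add crouch-switched rear cloth, body and pad-feather physics, with leg-driven front-cloth correction; bRigidBodyEnabled selects this or the cached pose.
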AnimBlueprint /BRCosmetics/Characters/Player/Female/Medium/Bodies/F_MED_Angel/Meshes/F_MED_Angel_AnimBP_Nativized_OLD
target_skeleton=&/F_MED_Angel_Skeleton
graphs [
 AnimGraph {nodes [
   "Transform (Modify) Bone - Bone: dyn_apron_frnt_rb_02_l"(Alpha=Get SpineRotLeft() -> (SpineRotLeft:float), Translation="0,0,0", Rotation="0,-20,0", Scale="1,1,1") {
    rotationMode=BMM_Additive
    rotationSpace=BCS_BoneSpace
    translationMode=BMM_Additive
    translationSpace=BCS_BoneSpace
    alphaBoolBlend {blendOption=Linear}
    alphaScaleBiasClamp {bInterpResult=true, interpSpeedDecreasing=5, interpSpeedIncreasing=5}
    boneToModify {boneName=dyn_apron_frnt_rb_02_l}
    rotation {yaw=-20}
   }
   RigidBody() {
    bOverrideWorldGravity=true
    overridePhysicsAsset=&/F_MED_Angel_FrontCloth_Crouch_Physics
    simulationSpace=BaseBoneSpace
    alphaBoolBlend {blendOption=Linear}
    baseBoneRef {boneName=pelvis}
    componentAppliedLinearAccClamp=<500 500 500>
    componentLinearAccScale=<1 0.7 0.7>
    componentLinearVelScale=<1 0.7 0.7>
    overrideWorldGravity {z=-980}
   }
   n1 = RigidBody!(
    ComponentPose=Local To Component(LocalPose=Use cached pose 'AnimCache'())
    Alpha=Get ClothAlpha() -> (ClothAlpha:float)
    ExternalForce
    ComponentLinearVelScale:Vector="1,1,1"
   ) -> (Pose) {
    bOverrideWorldGravity=true
    overridePhysicsAsset=&/F_MED_Angel_RearCloth_Physics
    simulationSpace=BaseBoneSpace
    alphaBoolBlend {blendOption=Linear}
    alphaScaleBiasClamp {bInterpResult=true, interpSpeedDecreasing=1, interpSpeedIncreasing=1}
    baseBoneRef {boneName=root}
    componentAppliedLinearAccClamp=<10 10 10>
    componentLinearAccScale=<0.7 0.7 0.7>
    componentLinearVelScale=<1 1 1>
    overrideWorldGravity {z=-950}
   }
   n2 = Component To Local(ComponentPose=@n1)
   n3 = RigidBody(ComponentPose=Local To Component(LocalPose=Use cached pose 'AnimCache'()), ComponentLinearVelScale="1,1,1") {
    bOverrideWorldGravity=true
    overridePhysicsAsset=&/F_MED_Angel_RearCloth_Crouch_Physics
    simulationSpace=BaseBoneSpace
    alphaBoolBlend {blendOption=Linear}
    alphaScaleBiasClamp {bInterpResult=true, interpSpeedDecreasing=1, interpSpeedIncreasing=1}
    baseBoneRef {boneName=root}
    componentAppliedLinearAccClamp=<10 10 10>
    componentLinearAccScale=<0.7 0.7 0.7>
    componentLinearVelScale=<1 1 1>
    overrideWorldGravity {z=-980}
   }
   n4 = Component To Local(ComponentPose=@n3)
   n5 = Get bIsCrouching()
   n6 = "Constraint - dyn_cloth_font_mid_01 (PP)"!(
    ComponentPose=Local To Component(LocalPose=Blend Poses by bool(BlendPose_0=@n4, BlendPose_1=@n2, bActiveValue=@n5) {blendType=Linear})
    Alpha=1
    ConstraintWeights_0=0.5
    ConstraintWeights_1:float=0.5
   ) -> (Pose) {
    alphaBoolBlend {blendOption=Linear}
    boneToModify {boneName=dyn_cloth_font_mid_01}
    constraintSetup [{targetBone {boneName=thigh_twist_01_l}}, {targetBone {boneName=thigh_twist_01_r}}]
    constraintWeights [1,1]
   }
   n7 = "Transform (Modify) Bone - Bone: dyn_cloth_font_mid_01"(ComponentPose=@n6, Alpha=Get leg_rot_l() -> (leg_rot_l:double), Translation="2,5,0", Rotation="-10,0,60", Scale="1,1,1") {
    rotationMode=BMM_Additive
    rotationSpace=BCS_BoneSpace
    translationMode=BMM_Additive
    translationSpace=BCS_BoneSpace
    alphaBoolBlend {blendOption=Linear}
    alphaScaleBiasClamp {bInterpResult=true, interpSpeedDecreasing=9, interpSpeedIncreasing=9}
    boneToModify {boneName=dyn_cloth_font_mid_01}
    rotation {pitch=-10, roll=60}
    translation=<2 5>
   }
   n8 = "Transform (Modify) Bone - Bone: dyn_cloth_font_mid_01"(ComponentPose=@n7, Alpha=Get leg_rot_r() -> (leg_rot_r:double), Translation="2,-5,0", Rotation="-10,0,-60", Scale="1,1,1") {
    rotationMode=BMM_Additive
    rotationSpace=BCS_BoneSpace
    translationMode=BMM_Additive
    translationSpace=BCS_BoneSpace
    alphaBoolBlend {blendOption=Linear}
    alphaScaleBiasClamp {bInterpResult=true, interpSpeedDecreasing=9, interpSpeedIncreasing=9}
    boneToModify {boneName=dyn_cloth_font_mid_01}
    rotation {pitch=-10, roll=-60}
    translation=<2 -5>
   }
   n9 = RigidBody!(ComponentPose=@n8, Alpha, ExternalForce) -> (Pose) {
    bClampLinearTranslationLimitToRefPose=true
    bOverrideWorldGravity=true
    overridePhysicsAsset=&/F_MED_Angel_Physics
    simulationSpace=BaseBoneSpace
    alphaBoolBlend {blendOption=Linear}
    baseBoneRef {boneName=spine_02}
    componentAppliedLinearAccClamp=<1200 1200 1200>
    componentLinearAccScale=<1 0.7 0.7>
    componentLinearVelScale=<1 0.7 0.7>
    overrideWorldGravity {z=-980}
   }
   n10 = RigidBody(ComponentPose=@n9) {
    bClampLinearTranslationLimitToRefPose=true
    bOverrideWorldGravity=true
    overridePhysicsAsset=&/F_MED_Angel_PadFeathers_Physics
    simulationSpace=BaseBoneSpace
    alphaBoolBlend {blendOption=Linear}
    baseBoneRef {boneName=spine_05}
    componentAppliedLinearAccClamp=<500 500 500>
    componentLinearAccScale=<1 1 1>
    componentLinearVelScale=<1 1 1>
    overrideWorldGravity {z=-980}
   }
   Output Pose(Result=Blend Poses by bool(BlendPose_0=Component To Local(ComponentPose=@n10), BlendPose_1=Use cached pose 'AnimCache'(), bActiveValue=Get bRigidBodyEnabled()) {blendType=Linear}) {layerGroup=None}
   n11 = Get RearClothAlpha() -> (RearClothAlpha:float)
   n12 = Anim Dynamics - Bone: dyn_skirt_side_01_r(
    ComponentPose=Local To Component(LocalPose=Copy Pose From Mesh() {bCopyCurves=true, bUseAttachedParent=true})
    Alpha=1
    ExternalForce="0,0,0"
   ) {
    angularDampingOverride=0.9999
    bGravityOverrideInSimSpace=true
    bOverrideAngularDamping=true
    gravityScale=0.35
    simulationSpace=BoneRelative
    alphaBoolBlend {blendOption=Linear}
    boundBone {boneName=dyn_skirt_side_01_r}
    componentAppliedLinearAccClamp=<50 50 50>
    physicsBodyDefinitions [{
      sphereCollisionRadius=0
      boundBone {boneName=dyn_skirt_side_01_r}
      boxExtents=<5 20 20>
      constraintSetup {angularLimitsMax=<0 15 0>, angularLimitsMin=<0 0 0>, angularTarget=<0 0 0>}
      localJointOffset=<5 0 0>
     }]
    relativeSpaceBone {boneName=root}
   }
   n13 = Anim Dynamics - Bone: dyn_skirt_side_02_r(ComponentPose=@n12, Alpha=1, ExternalForce="0,0,0") {
    angularDampingOverride=0.9999
    bGravityOverrideInSimSpace=true
    bOverrideAngularDamping=true
    gravityScale=0.35
    simulationSpace=BoneRelative
    alphaBoolBlend {blendOption=Linear}
    boundBone {boneName=dyn_skirt_side_02_r}
    componentAppliedLinearAccClamp=<50 50 50>
    physicsBodyDefinitions [{
      sphereCollisionRadius=0
      boundBone {boneName=dyn_skirt_side_02_r}
      boxExtents=<5 20 20>
      constraintSetup {angularLimitsMax=<0 20 10>, angularLimitsMin=<0 0 -10>, angularTarget=<0 0 0>}
      localJointOffset=<5 0 0>
     }]
    relativeSpaceBone {boneName=root}
   }
   n14 = Anim Dynamics - Bone: dyn_skirt_side_03_r(ComponentPose=@n13, Alpha=1, ExternalForce="0,0,0") {
    angularDampingOverride=0.9999
    bGravityOverrideInSimSpace=true
    bOverrideAngularDamping=true
    gravityScale=0.35
    simulationSpace=BoneRelative
    alphaBoolBlend {blendOption=Linear}
    boundBone {boneName=dyn_skirt_side_03_r}
    componentAppliedLinearAccClamp=<50 50 50>
    physicsBodyDefinitions [{
      sphereCollisionRadius=0
      boundBone {boneName=dyn_skirt_side_03_r}
      boxExtents=<5 20 20>
      constraintSetup {angularLimitsMax=<0 30 10>, angularLimitsMin=<0 0 -10>, angularTarget=<0 0 0>}
      localJointOffset=<5 0 0>
     }]
    relativeSpaceBone {boneName=root}
   }
   n15 = Anim Dynamics - Bone: dyn_skirt_sideMid_01_r(ComponentPose=@n14, Alpha=1, ExternalForce="0,0,0") {
    angularDampingOverride=0.9999
    bGravityOverrideInSimSpace=true
    bOverrideAngularDamping=true
    gravityScale=0.35
    simulationSpace=BoneRelative
    alphaBoolBlend {blendOption=Linear}
    boundBone {boneName=dyn_skirt_sideMid_01_r}
    componentAppliedLinearAccClamp=<50 50 50>
    physicsBodyDefinitions [{
      sphereCollisionRadius=0
      boundBone {boneName=dyn_skirt_sideMid_01_r}
      boxExtents=<5 20 20>
      constraintSetup {angularLimitsMax=<0 15 0>, angularLimitsMin=<0 0 0>, angularTarget=<0 0 0>}
      localJointOffset=<5 0 0>
     }]
    relativeSpaceBone {boneName=root}
   }
   n16 = Anim Dynamics - Bone: dyn_skirt_sideMid_02_r(ComponentPose=@n15, Alpha=1, ExternalForce="0,0,0") {
    angularDampingOverride=0.9999
    bGravityOverrideInSimSpace=true
    bOverrideAngularDamping=true
    gravityScale=0.35
    simulationSpace=BoneRelative
    alphaBoolBlend {blendOption=Linear}
    boundBone {boneName=dyn_skirt_sideMid_02_r}
    componentAppliedLinearAccClamp=<50 50 50>
    physicsBodyDefinitions [{
      sphereCollisionRadius=0
      boundBone {boneName=dyn_skirt_sideMid_02_r}
      boxExtents=<5 20 20>
      constraintSetup {angularLimitsMax=<0 20 10>, angularLimitsMin=<0 0 -10>, angularTarget=<0 0 0>}
      localJointOffset=<5 0 0>
     }]
    relativeSpaceBone {boneName=root}
   }
   n17 = Anim Dynamics - Bone: dyn_skirt_sideMid_03_r(ComponentPose=@n16, Alpha=1, ExternalForce="0,0,0") {
    angularDampingOverride=0.9999
    bGravityOverrideInSimSpace=true
    bOverrideAngularDamping=true
    gravityScale=0.35
    simulationSpace=BoneRelative
    alphaBoolBlend {blendOption=Linear}
    boundBone {boneName=dyn_skirt_sideMid_03_r}
    componentAppliedLinearAccClamp=<50 50 50>
    physicsBodyDefinitions [{
      sphereCollisionRadius=0
      boundBone {boneName=dyn_skirt_sideMid_03_r}
      boxExtents=<5 20 20>
      constraintSetup {angularLimitsMax=<0 30 10>, angularLimitsMin=<0 0 -10>, angularTarget=<0 0 0>}
      localJointOffset=<5 0 0>
     }]
    planarLimits [{drivingBone {boneName=calf_r}, planeTransform {location=<0 10 0>, rotation=<rot 0 0 90>}}]
    relativeSpaceBone {boneName=root}
   }
   n18 = Anim Dynamics - Bone: dyn_skirt_frontMid_01_r(ComponentPose=@n17, Alpha=1, ExternalForce="0,0,0") {
    angularDampingOverride=0.9999
    bGravityOverrideInSimSpace=true
    bOverrideAngularDamping=true
    gravityScale=0.35
    simulationSpace=BoneRelative
    alphaBoolBlend {blendOption=Linear}
    boundBone {boneName=dyn_skirt_frontMid_01_r}
    componentAppliedLinearAccClamp=<50 50 50>
    physicsBodyDefinitions [{
      sphereCollisionRadius=0
      boundBone {boneName=dyn_skirt_frontMid_01_r}
      boxExtents=<5 20 20>
      constraintSetup {angularLimitsMax=<0 15 0>, angularLimitsMin=<0 0 0>, angularTarget=<0 0 0>}
      localJointOffset=<5 0 0>
     }]
    relativeSpaceBone {boneName=root}
   }
   n19 = Anim Dynamics - Bone: dyn_skirt_frontMid_02_r(ComponentPose=@n18, Alpha=1, ExternalForce="0,0,0") {
    angularDampingOverride=0.9999
    bGravityOverrideInSimSpace=true
    bOverrideAngularDamping=true
    gravityScale=0.35
    simulationSpace=BoneRelative
    alphaBoolBlend {blendOption=Linear}
    boundBone {boneName=dyn_skirt_frontMid_02_r}
    componentAppliedLinearAccClamp=<50 50 50>
    physicsBodyDefinitions [{
      sphereCollisionRadius=0
      boundBone {boneName=dyn_skirt_frontMid_02_r}
      boxExtents=<5 20 20>
      constraintSetup {angularLimitsMax=<0 15 0>, angularLimitsMin=<0 0 0>, angularTarget=<0 0 0>}
      localJointOffset=<5 0 0>
     }]
    relativeSpaceBone {boneName=root}
   }
   n20 = Anim Dynamics - Bone: dyn_skirt_frontMid_03_r(ComponentPose=@n19, Alpha=1, ExternalForce="0,0,0") {
    angularDampingOverride=0.9999
    bGravityOverrideInSimSpace=true
    bOverrideAngularDamping=true
    gravityScale=0.35
    simulationSpace=BoneRelative
    alphaBoolBlend {blendOption=Linear}
    boundBone {boneName=dyn_skirt_frontMid_03_r}
    componentAppliedLinearAccClamp=<50 50 50>
    physicsBodyDefinitions [{
      sphereCollisionRadius=0
      boundBone {boneName=dyn_skirt_frontMid_03_r}
      boxExtents=<5 20 20>
      constraintSetup {angularLimitsMax=<0 15 0>, angularLimitsMin=<0 0 0>, angularTarget=<0 0 0>}
      localJointOffset=<5 0 0>
     }]
    relativeSpaceBone {boneName=root}
   }
   n21 = "Constraint - dyn_skirt_frontMid_01_l (P)"(ComponentPose=@n20, Alpha=1, ConstraintWeights_0=1) {
    alphaBoolBlend {blendOption=Linear}
    boneToModify {boneName=dyn_skirt_frontMid_01_l}
    constraintSetup [{targetBone {boneName=thigh_l}}]
    constraintWeights [1]
   }
   n22 = Get CrouchAlpha() -> (CrouchAlpha:float)
   n23 = Anim Dynamics - Bone: dyn_skirt_side_01_l(ComponentPose=@n21, Alpha=@n22, ExternalForce="0,0,0") {
    angularDampingOverride=0.9999
    bGravityOverrideInSimSpace=true
    bOverrideAngularDamping=true
    gravityScale=0.35
    simulationSpace=BoneRelative
    alphaBoolBlend {blendOption=Linear}
    boundBone {boneName=dyn_skirt_side_01_l}
    componentAppliedLinearAccClamp=<50 50 50>
    physicsBodyDefinitions [{
      sphereCollisionRadius=0
      boundBone {boneName=dyn_skirt_side_01_l}
      boxExtents=<5 20 20>
      constraintSetup {angularLimitsMax=<0 15 0>, angularLimitsMin=<0 0 0>, angularTarget=<0 0 0>}
      localJointOffset=<5 0 0>
     }]
    relativeSpaceBone {boneName=root}
   }
   n24 = Anim Dynamics - Bone: dyn_skirt_side_02_l(ComponentPose=@n23, Alpha=@n22, ExternalForce="0,0,0") {
    angularDampingOverride=0.9999
    bGravityOverrideInSimSpace=true
    bOverrideAngularDamping=true
    gravityScale=0.35
    simulationSpace=BoneRelative
    alphaBoolBlend {blendOption=Linear}
    boundBone {boneName=dyn_skirt_side_02_l}
    componentAppliedLinearAccClamp=<50 50 50>
    physicsBodyDefinitions [{
      sphereCollisionRadius=0
      boundBone {boneName=dyn_skirt_side_02_l}
      boxExtents=<5 20 20>
      constraintSetup {angularLimitsMax=<0 20 10>, angularLimitsMin=<0 0 -10>, angularTarget=<0 0 0>}
      localJointOffset=<5 0 0>
     }]
    relativeSpaceBone {boneName=root}
   }
   n25 = Anim Dynamics - Bone: dyn_skirt_side_03_l(ComponentPose=@n24, Alpha=@n22, ExternalForce="0,0,0") {
    angularDampingOverride=0.9999
    bGravityOverrideInSimSpace=true
    bOverrideAngularDamping=true
    gravityScale=0.35
    simulationSpace=BoneRelative
    alphaBoolBlend {blendOption=Linear}
    boundBone {boneName=dyn_skirt_side_03_l}
    componentAppliedLinearAccClamp=<50 50 50>
    physicsBodyDefinitions [{
      sphereCollisionRadius=0
      boundBone {boneName=dyn_skirt_side_03_l}
      boxExtents=<5 20 20>
      constraintSetup {angularLimitsMax=<0 30 10>, angularLimitsMin=<0 0 -10>, angularTarget=<0 0 0>}
      localJointOffset=<5 0 0>
     }]
    relativeSpaceBone {boneName=root}
   }
   n26 = Anim Dynamics - Bone: dyn_skirt_sideMid_01_l(ComponentPose=@n25, Alpha=1, ExternalForce="0,0,0") {
    angularDampingOverride=0.9999
    bGravityOverrideInSimSpace=true
    bOverrideAngularDamping=true
    gravityScale=0.35
    simulationSpace=BoneRelative
    alphaBoolBlend {blendOption=Linear}
    boundBone {boneName=dyn_skirt_sideMid_01_l}
    componentAppliedLinearAccClamp=<50 50 50>
    physicsBodyDefinitions [{
      sphereCollisionRadius=0
      boundBone {boneName=dyn_skirt_sideMid_01_l}
      boxExtents=<5 20 20>
      constraintSetup {angularLimitsMax=<0 15 0>, angularLimitsMin=<0 0 0>, angularTarget=<0 0 0>}
      localJointOffset=<5 0 0>
     }]
    relativeSpaceBone {boneName=root}
   }
   n27 = Anim Dynamics - Bone: dyn_skirt_sideMid_02_l(ComponentPose=@n26, Alpha=1, ExternalForce="0,0,0") {
    angularDampingOverride=0.9999
    bGravityOverrideInSimSpace=true
    bOverrideAngularDamping=true
    gravityScale=0.35
    simulationSpace=BoneRelative
    alphaBoolBlend {blendOption=Linear}
    boundBone {boneName=dyn_skirt_sideMid_02_l}
    componentAppliedLinearAccClamp=<50 50 50>
    physicsBodyDefinitions [{
      sphereCollisionRadius=0
      boundBone {boneName=dyn_skirt_sideMid_02_l}
      boxExtents=<5 20 20>
      constraintSetup {angularLimitsMax=<0 20 10>, angularLimitsMin=<0 0 -10>, angularTarget=<0 0 0>}
      localJointOffset=<5 0 0>
     }]
    relativeSpaceBone {boneName=root}
   }
   n28 = Anim Dynamics - Bone: dyn_skirt_sideMid_03_l(ComponentPose=@n27, Alpha=1, ExternalForce="0,0,0") {
    angularDampingOverride=0.9999
    bGravityOverrideInSimSpace=true
    bOverrideAngularDamping=true
    gravityScale=0.35
    simulationSpace=BoneRelative
    alphaBoolBlend {blendOption=Linear}
    boundBone {boneName=dyn_skirt_sideMid_03_l}
    componentAppliedLinearAccClamp=<50 50 50>
    physicsBodyDefinitions [{
      sphereCollisionRadius=0
      boundBone {boneName=dyn_skirt_sideMid_03_l}
      boxExtents=<5 20 20>
      constraintSetup {angularLimitsMax=<0 30 10>, angularLimitsMin=<0 0 -10>, angularTarget=<0 0 0>}
      localJointOffset=<5 0 0>
     }]
    planarLimits [{drivingBone {boneName=calf_l}, planeTransform {location=<0 -10 0>, rotation=<rot 0 0 -90>}}]
    relativeSpaceBone {boneName=root}
   }
   n29 = Anim Dynamics - Bone: dyn_skirt_frontMid_01_l(ComponentPose=@n28, Alpha=1, ExternalForce="0,0,0") {
    angularDampingOverride=0.9999
    bGravityOverrideInSimSpace=true
    bOverrideAngularDamping=true
    gravityScale=0.35
    simulationSpace=BoneRelative
    alphaBoolBlend {blendOption=Linear}
    boundBone {boneName=dyn_skirt_frontMid_01_l}
    componentAppliedLinearAccClamp=<50 50 50>
    physicsBodyDefinitions [{
      sphereCollisionRadius=0
      boundBone {boneName=dyn_skirt_frontMid_01_l}
      boxExtents=<5 20 20>
      constraintSetup {angularLimitsMax=<0 15 0>, angularLimitsMin=<0 0 0>, angularTarget=<0 0 0>}
      localJointOffset=<5 0 0>
     }]
    relativeSpaceBone {boneName=root}
   }
   n30 = Anim Dynamics - Bone: dyn_skirt_frontMid_02_l(ComponentPose=@n29, Alpha=1, ExternalForce="0,0,0") {
    angularDampingOverride=0.9999
    bGravityOverrideInSimSpace=true
    bOverrideAngularDamping=true
    gravityScale=0.35
    simulationSpace=BoneRelative
    alphaBoolBlend {blendOption=Linear}
    boundBone {boneName=dyn_skirt_frontMid_02_l}
    componentAppliedLinearAccClamp=<50 50 50>
    physicsBodyDefinitions [{
      sphereCollisionRadius=0
      boundBone {boneName=dyn_skirt_frontMid_02_l}
      boxExtents=<5 20 20>
      constraintSetup {angularLimitsMax=<0 15 0>, angularLimitsMin=<0 0 0>, angularTarget=<0 0 0>}
      localJointOffset=<5 0 0>
     }]
    relativeSpaceBone {boneName=root}
   }
   n31 = Anim Dynamics - Bone: dyn_skirt_frontMid_03_l(ComponentPose=@n30, Alpha=1, ExternalForce="0,0,0") {
    angularDampingOverride=0.9999
    bGravityOverrideInSimSpace=true
    bOverrideAngularDamping=true
    gravityScale=0.35
    simulationSpace=BoneRelative
    alphaBoolBlend {blendOption=Linear}
    boundBone {boneName=dyn_skirt_frontMid_03_l}
    componentAppliedLinearAccClamp=<50 50 50>
    physicsBodyDefinitions [{
      sphereCollisionRadius=0
      boundBone {boneName=dyn_skirt_frontMid_03_l}
      boxExtents=<5 20 20>
      constraintSetup {angularLimitsMax=<0 15 0>, angularLimitsMin=<0 0 0>, angularTarget=<0 0 0>}
      localJointOffset=<5 0 0>
     }]
    relativeSpaceBone {boneName=root}
   }
   n32 = Anim Dynamics - Bone: dyn_calf_fthr_01_l(ComponentPose=@n31, Alpha=0.3, ExternalForce="0,0,0") {
    angularDampingOverride=0.985
    angularSpringConstant=500
    bAngularSpring=true
    bGravityOverrideInSimSpace=true
    bLinearSpring=true
    bOverrideAngularDamping=true
    bOverrideLinearDamping=true
    gravityScale=0.085
    linearDampingOverride=0.99
    linearSpringConstant=600
    alphaBoolBlend {blendOption=Linear}
    boundBone {boneName=dyn_calf_fthr_01_l}
    componentAppliedLinearAccClamp=<1000 1000 1000>
    componentLinearVelScale=<100 100 100>
    physicsBodyDefinitions [{
      sphereCollisionRadius=0
      boundBone {boneName=dyn_calf_fthr_01_l}
      boxExtents=<30 30 30>
      constraintSetup {
       angularTargetAxis=AxisY
       angularLimitsMax=<0 0 0>
       angularLimitsMin=<0 0 0>
       angularTarget=<0 0 0>
       linearAxesMax=<2 2 1>
       linearAxesMin=<-2 -2 -1>
      }
      localJointOffset=<0 0 0>
     }]
   }
   n33 = Anim Dynamics - Bone: dyn_calf_fthr_02_l(ComponentPose=@n32, Alpha=0.35, ExternalForce="0,0,0") {
    angularDampingOverride=0.985
    angularSpringConstant=300
    bAngularSpring=true
    bGravityOverrideInSimSpace=true
    bLinearSpring=true
    bOverrideAngularDamping=true
    bOverrideLinearDamping=true
    gravityScale=0.15
    linearDampingOverride=0.99
    linearSpringConstant=200
    alphaBoolBlend {blendOption=Linear}
    boundBone {boneName=dyn_calf_fthr_02_l}
    componentAppliedLinearAccClamp=<1000 1000 1000>
    componentLinearVelScale=<100 100 100>
    physicsBodyDefinitions [{
      sphereCollisionRadius=0
      boundBone {boneName=dyn_calf_fthr_02_l}
      boxExtents=<30 45 7>
      constraintSetup {
       angularTargetAxis=AxisY
       angularLimitsMax=<0 0 0>
       angularLimitsMin=<0 0 0>
       angularTarget=<0 0 0>
       linearAxesMax=<2 2 1>
       linearAxesMin=<-2 -2 -1>
      }
      localJointOffset=<0 0 0>
     }]
   }
   n34 = Anim Dynamics - Bone: dyn_calf_fthr_03_l(ComponentPose=@n33, Alpha=0.4, ExternalForce="0,0,0") {
    angularDampingOverride=0.985
    angularSpringConstant=500
    bAngularSpring=true
    bGravityOverrideInSimSpace=true
    bLinearSpring=true
    bOverrideAngularDamping=true
    bOverrideLinearDamping=true
    gravityScale=0.3
    linearDampingOverride=0.99
    linearSpringConstant=500
    alphaBoolBlend {blendOption=Linear}
    boundBone {boneName=dyn_calf_fthr_03_l}
    componentAppliedLinearAccClamp=<1000 1000 1000>
    componentLinearVelScale=<100 100 100>
    physicsBodyDefinitions [{
      sphereCollisionRadius=0
      boundBone {boneName=dyn_calf_fthr_03_l}
      boxExtents=<20 10 15>
      constraintSetup {
       angularTargetAxis=AxisY
       angularLimitsMax=<0 0 0>
       angularLimitsMin=<0 0 0>
       angularTarget=<0 0 0>
       linearAxesMax=<2 2 1>
       linearAxesMin=<-2 -2 -1>
      }
      localJointOffset=<0 0 0>
     }]
   }
   n35 = Anim Dynamics - Bone: dyn_calf_fthr_04_l(ComponentPose=@n34, Alpha=0.4, ExternalForce="0,0,0") {
    angularDampingOverride=0.985
    angularSpringConstant=500
    bAngularSpring=true
    bGravityOverrideInSimSpace=true
    bLinearSpring=true
    bOverrideAngularDamping=true
    bOverrideLinearDamping=true
    gravityScale=0.085
    linearDampingOverride=0.99
    linearSpringConstant=600
    alphaBoolBlend {blendOption=Linear}
    boundBone {boneName=dyn_calf_fthr_04_l}
    componentAppliedLinearAccClamp=<1000 1000 1000>
    componentLinearVelScale=<100 100 100>
    physicsBodyDefinitions [{
      sphereCollisionRadius=0
      boundBone {boneName=dyn_calf_fthr_04_l}
      boxExtents=<30 30 30>
      constraintSetup {
       angularTargetAxis=AxisY
       angularLimitsMax=<0 0 0>
       angularLimitsMin=<0 0 0>
       angularTarget=<0 0 0>
       linearAxesMax=<2 2 1>
       linearAxesMin=<-2 -2 -1>
      }
      localJointOffset=<0 0 0>
     }]
   }
   n36 = Anim Dynamics - Bone: dyn_calf_fthr_05_l(ComponentPose=@n35, Alpha=0.32, ExternalForce="0,0,0") {
    angularDampingOverride=0.985
    angularSpringConstant=300
    bAngularSpring=true
    bGravityOverrideInSimSpace=true
    bLinearSpring=true
    bOverrideAngularDamping=true
    bOverrideLinearDamping=true
    gravityScale=0.15
    linearDampingOverride=0.99
    linearSpringConstant=200
    alphaBoolBlend {blendOption=Linear}
    boundBone {boneName=dyn_calf_fthr_05_l}
    componentAppliedLinearAccClamp=<1000 1000 1000>
    componentLinearVelScale=<100 100 100>
    physicsBodyDefinitions [{
      sphereCollisionRadius=0
      boundBone {boneName=dyn_calf_fthr_05_l}
      boxExtents=<30 45 7>
      constraintSetup {
       angularTargetAxis=AxisY
       angularLimitsMax=<0 0 0>
       angularLimitsMin=<0 0 0>
       angularTarget=<0 0 0>
       linearAxesMax=<1.5 1.5 0.5>
       linearAxesMin=<-1.5 -1.5 -0.5>
      }
      localJointOffset=<0 0 0>
     }]
   }
   n37 = Anim Dynamics - Bone: dyn_calf_fthr_01_r(ComponentPose=@n36, Alpha=0.36, ExternalForce="0,0,0") {
    angularDampingOverride=0.985
    angularSpringConstant=500
    bAngularSpring=true
    bGravityOverrideInSimSpace=true
    bLinearSpring=true
    bOverrideAngularDamping=true
    bOverrideLinearDamping=true
    gravityScale=0.3
    linearDampingOverride=0.99
    linearSpringConstant=500
    alphaBoolBlend {blendOption=Linear}
    boundBone {boneName=dyn_calf_fthr_01_r}
    componentAppliedLinearAccClamp=<1000 1000 1000>
    componentLinearVelScale=<100 100 100>
    physicsBodyDefinitions [{
      sphereCollisionRadius=0
      boundBone {boneName=dyn_calf_fthr_01_r}
      boxExtents=<20 10 15>
      constraintSetup {
       angularTargetAxis=AxisY
       angularLimitsMax=<0 0 0>
       angularLimitsMin=<0 0 0>
       angularTarget=<0 0 0>
       linearAxesMax=<2 2 1>
       linearAxesMin=<-2 -2 -1>
      }
      localJointOffset=<0 0 0>
     }]
   }
   n38 = Anim Dynamics - Bone: dyn_calf_fthr_02_r(ComponentPose=@n37, Alpha=0.4, ExternalForce="0,0,0") {
    angularDampingOverride=0.985
    angularSpringConstant=500
    bAngularSpring=true
    bGravityOverrideInSimSpace=true
    bLinearSpring=true
    bOverrideAngularDamping=true
    bOverrideLinearDamping=true
    gravityScale=0.085
    linearDampingOverride=0.99
    linearSpringConstant=600
    alphaBoolBlend {blendOption=Linear}
    boundBone {boneName=dyn_calf_fthr_02_r}
    componentAppliedLinearAccClamp=<1000 1000 1000>
    componentLinearVelScale=<100 100 100>
    physicsBodyDefinitions [{
      sphereCollisionRadius=0
      boundBone {boneName=dyn_calf_fthr_02_r}
      boxExtents=<30 30 30>
      constraintSetup {
       angularTargetAxis=AxisY
       angularLimitsMax=<0 0 0>
       angularLimitsMin=<0 0 0>
       angularTarget=<0 0 0>
       linearAxesMax=<2 2 1>
       linearAxesMin=<-2 -2 -1>
      }
      localJointOffset=<0 0 0>
     }]
   }
   n39 = Anim Dynamics - Bone: dyn_calf_fthr_03_r(ComponentPose=@n38, Alpha=0.45, ExternalForce="0,0,0") {
    angularDampingOverride=0.985
    angularSpringConstant=300
    bAngularSpring=true
    bGravityOverrideInSimSpace=true
    bLinearSpring=true
    bOverrideAngularDamping=true
    bOverrideLinearDamping=true
    gravityScale=0.15
    linearDampingOverride=0.99
    linearSpringConstant=200
    alphaBoolBlend {blendOption=Linear}
    boundBone {boneName=dyn_calf_fthr_03_r}
    componentAppliedLinearAccClamp=<1000 1000 1000>
    componentLinearVelScale=<100 100 100>
    physicsBodyDefinitions [{
      sphereCollisionRadius=0
      boundBone {boneName=dyn_calf_fthr_03_r}
      boxExtents=<30 45 7>
      constraintSetup {
       angularTargetAxis=AxisY
       angularLimitsMax=<0 0 0>
       angularLimitsMin=<0 0 0>
       angularTarget=<0 0 0>
       linearAxesMax=<2 2 1>
       linearAxesMin=<-2 -2 -1>
      }
      localJointOffset=<0 0 0>
     }]
   }
   n40 = Anim Dynamics - Bone: dyn_calf_fthr_04_r(ComponentPose=@n39, Alpha=0.3, ExternalForce="0,0,0") {
    angularDampingOverride=0.985
    angularSpringConstant=500
    bAngularSpring=true
    bGravityOverrideInSimSpace=true
    bLinearSpring=true
    bOverrideAngularDamping=true
    bOverrideLinearDamping=true
    gravityScale=0.3
    linearDampingOverride=0.99
    linearSpringConstant=500
    alphaBoolBlend {blendOption=Linear}
    boundBone {boneName=dyn_calf_fthr_04_r}
    componentAppliedLinearAccClamp=<1000 1000 1000>
    componentLinearVelScale=<100 100 100>
    physicsBodyDefinitions [{
      sphereCollisionRadius=0
      boundBone {boneName=dyn_calf_fthr_04_r}
      boxExtents=<20 10 15>
      constraintSetup {
       angularTargetAxis=AxisY
       angularLimitsMax=<0 0 0>
       angularLimitsMin=<0 0 0>
       angularTarget=<0 0 0>
       linearAxesMax=<2 2 1>
       linearAxesMin=<-2 -2 -1>
      }
      localJointOffset=<0 0 0>
     }]
   }
   n41 = Anim Dynamics - Bone: dyn_calf_fthr_05_r(ComponentPose=@n40, Alpha=0.32, ExternalForce="0,0,0") {
    angularDampingOverride=0.985
    angularSpringConstant=300
    bAngularSpring=true
    bGravityOverrideInSimSpace=true
    bLinearSpring=true
    bOverrideAngularDamping=true
    bOverrideLinearDamping=true
    gravityScale=0.15
    linearDampingOverride=0.99
    linearSpringConstant=200
    alphaBoolBlend {blendOption=Linear}
    boundBone {boneName=dyn_calf_fthr_05_r}
    componentAppliedLinearAccClamp=<1000 1000 1000>
    componentLinearVelScale=<100 100 100>
    physicsBodyDefinitions [{
      sphereCollisionRadius=0
      boundBone {boneName=dyn_calf_fthr_05_r}
      boxExtents=<30 45 7>
      constraintSetup {
       angularTargetAxis=AxisY
       angularLimitsMax=<0 0 0>
       angularLimitsMin=<0 0 0>
       angularTarget=<0 0 0>
       linearAxesMax=<1.5 1.5 0.5>
       linearAxesMin=<-1.5 -1.5 -0.5>
      }
      localJointOffset=<0 0 0>
     }]
   }
   n42 = "Transform (Modify) Bone - Bone: driver_skirt_rear_01_r"(ComponentPose=@n41, Alpha=@n11, Translation="0,0,0", Rotation="0,0,35", Scale="1,1,1") {
    rotationMode=BMM_Additive
    alphaBoolBlend {blendOption=Linear}
    alphaScaleBiasClamp {bInterpResult=true, interpSpeedDecreasing=1, interpSpeedIncreasing=1}
    boneToModify {boneName=driver_skirt_rear_01_r}
    rotation {roll=35}
   }
   n43 = "Transform (Modify) Bone - Bone: driver_skirt_rear_01_l"(ComponentPose=@n42, Alpha=@n11, Translation="0,0,0", Rotation="0,0.000,35", Scale="1,1,1") {
    rotationMode=BMM_Additive
    alphaBoolBlend {blendOption=Linear}
    alphaScaleBiasClamp {bInterpResult=true, interpSpeedDecreasing=1, interpSpeedIncreasing=1}
    boneToModify {boneName=driver_skirt_rear_01_l}
    rotation {roll=35}
   }
   n44 = "Transform (Modify) Bone - Bone: dyn_skirt_frontMid_01_r"(ComponentPose=@n43, Alpha=@n11, Translation="0,0,0", Rotation="-10,-0.000,0", Scale="1,1,1") {
    rotationMode=BMM_Additive
    alphaBoolBlend {blendOption=Linear}
    alphaScaleBiasClamp {bInterpResult=true, interpSpeedDecreasing=1, interpSpeedIncreasing=1}
    boneToModify {boneName=dyn_skirt_frontMid_01_r}
    rotation {pitch=-10}
   }
   n45 = Get SkydivingAlpha() -> (SkydivingAlpha:float)
   n46 = "Transform (Modify) Bone - Bone: driver_skirt_rear_01_r"(ComponentPose=@n44, Alpha=@n45, Translation="0,0,0", Rotation="0,0,15", Scale="1,1,1") {
    rotationMode=BMM_Additive
    alphaBoolBlend {blendOption=Linear}
    alphaScaleBiasClamp {bInterpResult=true, interpSpeedDecreasing=1, interpSpeedIncreasing=1}
    boneToModify {boneName=driver_skirt_rear_01_r}
    rotation {roll=15}
   }
   n47 = "Transform (Modify) Bone - Bone: driver_skirt_rear_01_l"(ComponentPose=@n46, Alpha=@n45, Translation="0,0,0", Rotation="0,0.000,15", Scale="1,1,1") {
    rotationMode=BMM_Additive
    alphaBoolBlend {blendOption=Linear}
    alphaScaleBiasClamp {bInterpResult=true, interpSpeedDecreasing=1, interpSpeedIncreasing=1}
    boneToModify {boneName=driver_skirt_rear_01_l}
    rotation {roll=15}
   }
   n48 = "Transform (Modify) Bone - Bone: dyn_skirt_rear_02_r"(ComponentPose=@n47, Alpha=@n45, Translation="0,0,0", Rotation="0,0,15", Scale="1,1,1") {
    rotationMode=BMM_Additive
    alphaBoolBlend {blendOption=Linear}
    alphaScaleBiasClamp {bInterpResult=true, interpSpeedDecreasing=1, interpSpeedIncreasing=1}
    boneToModify {boneName=dyn_skirt_rear_02_r}
    rotation {roll=15}
   }
   n49 = "Transform (Modify) Bone - Bone: dyn_skirt_rear_02_l"(ComponentPose=@n48, Alpha=@n45, Translation="0,0,0", Rotation="0,0.000,15", Scale="1,1,1") {
    rotationMode=BMM_Additive
    alphaBoolBlend {blendOption=Linear}
    alphaScaleBiasClamp {bInterpResult=true, interpSpeedDecreasing=1, interpSpeedIncreasing=1}
    boneToModify {boneName=dyn_skirt_rear_02_l}
    rotation {roll=15}
   }
   n50 = Get FeatherAlpha() -> (FeatherAlpha:float)
   n51 = "Transform (Modify) Bone - Bone: dyn_calf_base_l"(ComponentPose=@n49, Alpha=@n50, Translation="0,0,0", Rotation="-20,0.000,0.000", Scale="1,1,1") {
    rotationMode=BMM_Additive
    rotationSpace=BCS_BoneSpace
    translationSpace=BCS_BoneSpace
    alphaBoolBlend {blendOption=Linear}
    alphaScaleBiasClamp {bInterpResult=true, interpSpeedDecreasing=9, interpSpeedIncreasing=9}
    boneToModify {boneName=dyn_calf_base_l}
    rotation {pitch=-20}
   }
   n52 = "Transform (Modify) Bone - Bone: dyn_calf_base_r"(ComponentPose=@n51, Alpha=@n50, Translation="0,0,0", Rotation="20,0.000,0.000", Scale="1,1,1") {
    rotationMode=BMM_Additive
    rotationSpace=BCS_BoneSpace
    translationSpace=BCS_BoneSpace
    alphaBoolBlend {blendOption=Linear}
    alphaScaleBiasClamp {bInterpResult=true, interpSpeedDecreasing=9, interpSpeedIncreasing=9}
    boneToModify {boneName=dyn_calf_base_r}
    rotation {pitch=20}
   }
   n53 = Get CrouchSkirtAlpha() -> (CrouchSkirtAlpha:float)
   n54 = "Transform (Modify) Bone - Bone: dyn_skirt_sideMid_01_r"(ComponentPose=@n52, Alpha=@n53, Translation="0,0,0", Rotation="-30,0.000,0.000", Scale="1,1,1") {
    rotationMode=BMM_Additive
    rotationSpace=BCS_BoneSpace
    translationSpace=BCS_BoneSpace
    alphaBoolBlend {blendOption=Linear}
    alphaScaleBiasClamp {bInterpResult=true, interpSpeedDecreasing=9, interpSpeedIncreasing=9}
    boneToModify {boneName=dyn_skirt_sideMid_01_r}
    rotation {pitch=-30}
   }
   n55 = "Transform (Modify) Bone - Bone: dyn_skirt_sideMid_02_r"(ComponentPose=@n54, Alpha=@n53, Translation="0,0,0", Rotation="10,0.000,0.000", Scale="1,1,1") {
    rotationMode=BMM_Additive
    rotationSpace=BCS_BoneSpace
    translationSpace=BCS_BoneSpace
    alphaBoolBlend {blendOption=Linear}
    alphaScaleBiasClamp {bInterpResult=true, interpSpeedDecreasing=9, interpSpeedIncreasing=9}
    boneToModify {boneName=dyn_skirt_sideMid_02_r}
    rotation {pitch=10}
   }
   n56 = "Transform (Modify) Bone - Bone: dyn_skirt_sideMid_03_r"(ComponentPose=@n55, Alpha=@n53, Translation="0,0,0", Rotation="10,0,0.000", Scale="1,1,1") {
    rotationMode=BMM_Additive
    rotationSpace=BCS_BoneSpace
    translationSpace=BCS_BoneSpace
    alphaBoolBlend {blendOption=Linear}
    alphaScaleBiasClamp {bInterpResult=true, interpSpeedDecreasing=9, interpSpeedIncreasing=9}
    boneToModify {boneName=dyn_skirt_sideMid_03_r}
    rotation {pitch=10}
   }
   n57 = Get CrouchSkirtAlpha()
   n58 = "Transform (Modify) Bone - Bone: dyn_skirt_frontMid_01_l"(ComponentPose=@n56, Alpha=@n57, Translation="0,0,0", Rotation="-9.941,23.679,1.096", Scale="1,1,1") {
    rotationMode=BMM_Additive
    rotationSpace=BCS_BoneSpace
    translationSpace=BCS_BoneSpace
    alphaBoolBlend {blendOption=Linear}
    alphaScaleBiasClamp {bInterpResult=true, interpSpeedDecreasing=9, interpSpeedIncreasing=9}
    boneToModify {boneName=dyn_skirt_frontMid_01_l}
    rotation=<rot -9.941 23.679 1.096>
   }
   n59 = "Transform (Modify) Bone - Bone: dyn_skirt_frontMid_02_l"(ComponentPose=@n58, Alpha=@n57, Translation="0,0,0", Rotation="38.698,10.676,-11.091", Scale="1,1,1") {
    rotationMode=BMM_Additive
    rotationSpace=BCS_BoneSpace
    translationSpace=BCS_BoneSpace
    alphaBoolBlend {blendOption=Linear}
    alphaScaleBiasClamp {bInterpResult=true, interpSpeedDecreasing=9, interpSpeedIncreasing=9}
    boneToModify {boneName=dyn_skirt_frontMid_02_l}
    rotation=<rot 38.698 10.676 -11.091>
   }
   n60 = "Transform (Modify) Bone - Bone: dyn_skirt_frontMid_03_l"(ComponentPose=@n59, Alpha=@n57, Translation="0,0,0", Rotation="20.000,15.000,-25.510", Scale="1,1,1") {
    rotationMode=BMM_Additive
    rotationSpace=BCS_BoneSpace
    translationSpace=BCS_BoneSpace
    alphaBoolBlend {blendOption=Linear}
    alphaScaleBiasClamp {bInterpResult=true, interpSpeedDecreasing=9, interpSpeedIncreasing=9}
    boneToModify {boneName=dyn_skirt_frontMid_03_l}
    rotation=<rot 20 15 -25.51>
   }
   n61 = Get CrouchSkirtAlpha()
   n62 = "Transform (Modify) Bone - Bone: dyn_skirt_side_01_l"(ComponentPose=@n60, Alpha=@n61, Translation="0,0,0", Rotation="-10,25,0", Scale="1,1,1") {
    rotationMode=BMM_Additive
    rotationSpace=BCS_BoneSpace
    translationSpace=BCS_BoneSpace
    alphaBoolBlend {blendOption=Linear}
    alphaScaleBiasClamp {bInterpResult=true, interpSpeedDecreasing=9, interpSpeedIncreasing=9}
    boneToModify {boneName=dyn_skirt_side_01_l}
    rotation {pitch=-10, yaw=25}
   }
   n63 = "Transform (Modify) Bone - Bone: dyn_skirt_side_02_l"(ComponentPose=@n62, Alpha=@n61, Translation="0,0,0", Rotation="-10.025,33.747,-21.708", Scale="1,1,1") {
    rotationMode=BMM_Additive
    rotationSpace=BCS_BoneSpace
    translationSpace=BCS_BoneSpace
    alphaBoolBlend {blendOption=Linear}
    alphaScaleBiasClamp {bInterpResult=true, interpSpeedDecreasing=9, interpSpeedIncreasing=9}
    boneToModify {boneName=dyn_skirt_side_02_l}
    rotation=<rot -10.025 33.747 -21.708>
   }
   n64 = "Transform (Modify) Bone - Bone: dyn_skirt_side_03_l"(ComponentPose=@n63, Alpha=@n61, Translation="0,0,0", Rotation="20,10,0.000", Scale="1,1,1") {
    rotationMode=BMM_Additive
    rotationSpace=BCS_BoneSpace
    translationSpace=BCS_BoneSpace
    alphaBoolBlend {blendOption=Linear}
    alphaScaleBiasClamp {bInterpResult=true, interpSpeedDecreasing=9, interpSpeedIncreasing=9}
    boneToModify {boneName=dyn_skirt_side_03_l}
    rotation {pitch=20, yaw=10}
   }
   n65 = "Transform (Modify) Bone - Bone: dyn_skirt_sideMid_01_l"(ComponentPose=@n64, Alpha=@n61, Translation="0,0,0", Rotation="-19.882,35.834,-1.899", Scale="1,1,1") {
    rotationMode=BMM_Additive
    rotationSpace=BCS_BoneSpace
    translationSpace=BCS_BoneSpace
    alphaBoolBlend {blendOption=Linear}
    alphaScaleBiasClamp {bInterpResult=true, interpSpeedDecreasing=9, interpSpeedIncreasing=9}
    boneToModify {boneName=dyn_skirt_sideMid_01_l}
    rotation=<rot -19.882 35.834 -1.899>
   }
   n66 = "Transform (Modify) Bone - Bone: dyn_skirt_sideMid_02_l"(ComponentPose=@n65, Alpha=@n61, Translation="0,0,0", Rotation="-4.683,-18.689,-0.007", Scale="1,1,1") {
    rotationMode=BMM_Additive
    rotationSpace=BCS_BoneSpace
    translationSpace=BCS_BoneSpace
    alphaBoolBlend {blendOption=Linear}
    alphaScaleBiasClamp {bInterpResult=true, interpSpeedDecreasing=9, interpSpeedIncreasing=9}
    boneToModify {boneName=dyn_skirt_sideMid_02_l}
    rotation=<rot -4.683 -18.689 -0.007>
   }
   n67 = "Transform (Modify) Bone - Bone: dyn_skirt_sideMid_03_l"(ComponentPose=@n66, Alpha=@n61, Translation="0,0,0", Rotation="-8.772,13.586,-18.200", Scale="1,1,1") {
    rotationMode=BMM_Additive
    rotationSpace=BCS_BoneSpace
    translationSpace=BCS_BoneSpace
    alphaBoolBlend {blendOption=Linear}
    alphaScaleBiasClamp {bInterpResult=true, interpSpeedDecreasing=9, interpSpeedIncreasing=9}
    boneToModify {boneName=dyn_skirt_sideMid_03_l}
    rotation=<rot -8.772 13.5861 -18.2001>
   }
   Save cached pose 'AnimCache'(Pose=Component To Local(ComponentPose=@n67)) {cacheName=AnimCache}
   n68 = Get FrontClothCrouchAlpha() -> (FrontClothCrouchAlpha:float)
   n69 = Flap Node - Bone: dyn_apron_frnt_rb_01_l!(ComponentPose:Component Space Pose Link Structure, Alpha:float=@n68) -> (Pose:Component Space Pose Link Structure) {
    is=title FortAnimGraphNode_Flap
    bUseClamp=true
    targetClampMin=-10
    targetComponent=RotationZ
    alphaBoolBlend {blendOption=Linear}
    sourceDrives [
     {
      bUseRange=false
      drivingCurve=null
      multiplier=-0.6
      rangeMax=1
      rangeMin=-1
      remappedMax=1
      remappedMin=0
      sourceComponent=RotationZ
      useQuaternion=true
      sourceBone {boneName=thigh_l}
     }
     {
      bUseRange=false
      drivingCurve=null
      multiplier=1
      rangeMax=1
      rangeMin=-1
      remappedMax=1
      remappedMin=0
      sourceComponent=RotationZ
      useQuaternion=true
      sourceBone {boneName=thigh_r}
     }
    ]
    targetBone {boneName=dyn_apron_frnt_rb_01_l}
   }
   n70 = Flap Node - Bone: dyn_apron_frnt_rb_01_r!(ComponentPose:Component Space Pose Link Structure=@n69, Alpha:float=@n68) -> (Pose:Component Space Pose Link Structure) {
    is=title FortAnimGraphNode_Flap
    bUseClamp=true
    targetClampMin=-10
    targetComponent=RotationZ
    alphaBoolBlend {blendOption=Linear}
    sourceDrives [
     {
      bUseRange=false
      drivingCurve=null
      multiplier=-0.6
      rangeMax=1
      rangeMin=-1
      remappedMax=1
      remappedMin=0
      sourceComponent=RotationZ
      useQuaternion=true
      sourceBone {boneName=thigh_r}
     }
     {
      bUseRange=false
      drivingCurve=null
      multiplier=1
      rangeMax=1
      rangeMin=-1
      remappedMax=1
      remappedMin=0
      sourceComponent=RotationZ
      useQuaternion=true
      sourceBone {boneName=thigh_r}
     }
    ]
    targetBone {boneName=dyn_apron_frnt_rb_01_r}
   }
   n71 = Flap Node - Bone: dyn_apron_frnt_rb_02_l!(ComponentPose:Component Space Pose Link Structure=@n70, Alpha:float=@n68) -> (Pose:Component Space Pose Link Structure) {
    is=title FortAnimGraphNode_Flap
    bUseClamp=true
    targetClampMin=-10
    targetComponent=RotationZ
    alphaBoolBlend {blendOption=Linear}
    sourceDrives [
     {
      bUseRange=false
      drivingCurve=null
      multiplier=-0.55
      rangeMax=1
      rangeMin=-1
      remappedMax=1
      remappedMin=0
      sourceComponent=RotationZ
      useQuaternion=true
      sourceBone {boneName=thigh_l}
     }
     {
      bUseRange=false
      drivingCurve=null
      multiplier=1
      rangeMax=1
      rangeMin=-1
      remappedMax=1
      remappedMin=0
      sourceComponent=RotationZ
      useQuaternion=true
      sourceBone {boneName=thigh_r}
     }
    ]
    targetBone {boneName=dyn_apron_frnt_rb_02_l}
   }
   n72 = Flap Node - Bone: dyn_apron_frnt_rb_02_r!(ComponentPose:Component Space Pose Link Structure=@n71, Alpha:float=@n68) -> (Pose:Component Space Pose Link Structure) {
    is=title FortAnimGraphNode_Flap
    bUseClamp=true
    targetClampMin=-10
    targetComponent=RotationZ
    alphaBoolBlend {blendOption=Linear}
    sourceDrives [
     {
      bUseRange=false
      drivingCurve=null
      multiplier=-0.55
      rangeMax=1
      rangeMin=-1
      remappedMax=1
      remappedMin=0
      sourceComponent=RotationZ
      useQuaternion=true
      sourceBone {boneName=thigh_r}
     }
     {
      bUseRange=false
      drivingCurve=null
      multiplier=1
      rangeMax=1
      rangeMin=-1
      remappedMax=1
      remappedMin=0
      sourceComponent=RotationZ
      useQuaternion=true
      sourceBone {boneName=thigh_r}
     }
    ]
    targetBone {boneName=dyn_apron_frnt_rb_02_r}
   }
   n73 = Flap Node - Bone: dyn_apron_frnt_rb_03_l!(ComponentPose:Component Space Pose Link Structure=@n72, Alpha:float=@n68) -> (Pose:Component Space Pose Link Structure) {
    is=title FortAnimGraphNode_Flap
    bUseClamp=true
    targetClampMin=-10
    targetComponent=RotationZ
    alphaBoolBlend {blendOption=Linear}
    sourceDrives [
     {
      bUseRange=false
      drivingCurve=null
      multiplier=-0.45
      rangeMax=1
      rangeMin=-1
      remappedMax=1
      remappedMin=0
      sourceComponent=RotationZ
      useQuaternion=true
      sourceBone {boneName=thigh_l}
     }
     {
      bUseRange=false
      drivingCurve=null
      multiplier=1
      rangeMax=1
      rangeMin=-1
      remappedMax=1
      remappedMin=0
      sourceComponent=RotationZ
      useQuaternion=true
      sourceBone {boneName=thigh_r}
     }
    ]
    targetBone {boneName=dyn_apron_frnt_rb_03_l}
   }
   n74 = Flap Node - Bone: dyn_apron_frnt_rb_03_r!(ComponentPose:Component Space Pose Link Structure=@n73, Alpha:float=@n68) -> (Pose:Component Space Pose Link Structure) {
    is=title FortAnimGraphNode_Flap
    bUseClamp=true
    targetClampMin=-10
    targetComponent=RotationZ
    alphaBoolBlend {blendOption=Linear}
    sourceDrives [
     {
      bUseRange=false
      drivingCurve=null
      multiplier=-0.45
      rangeMax=1
      rangeMin=-1
      remappedMax=1
      remappedMin=0
      sourceComponent=RotationZ
      useQuaternion=true
      sourceBone {boneName=thigh_r}
     }
     {
      bUseRange=false
      drivingCurve=null
      multiplier=1
      rangeMax=1
      rangeMin=-1
      remappedMax=1
      remappedMin=0
      sourceComponent=RotationZ
      useQuaternion=true
      sourceBone {boneName=thigh_r}
     }
    ]
    targetBone {boneName=dyn_apron_frnt_rb_03_r}
   }
   n75 = Flap Node - Bone: dyn_apron_frnt_rb_04_l!(ComponentPose:Component Space Pose Link Structure=@n74, Alpha:float=@n68) -> (Pose:Component Space Pose Link Structure) {
    is=title FortAnimGraphNode_Flap
    bUseClamp=true
    targetClampMin=-10
    targetComponent=RotationZ
    alphaBoolBlend {blendOption=Linear}
    sourceDrives [
     {
      bUseRange=false
      drivingCurve=null
      multiplier=-0.2
      rangeMax=1
      rangeMin=-1
      remappedMax=1
      remappedMin=0
      sourceComponent=RotationZ
      useQuaternion=true
      sourceBone {boneName=thigh_l}
     }
     {
      bUseRange=false
      drivingCurve=null
      multiplier=1
      rangeMax=1
      rangeMin=-1
      remappedMax=1
      remappedMin=0
      sourceComponent=RotationZ
      useQuaternion=true
      sourceBone {boneName=thigh_r}
     }
    ]
    targetBone {boneName=dyn_apron_frnt_rb_04_l}
   }
   Flap Node - Bone: dyn_apron_frnt_rb_04_r!(ComponentPose:Component Space Pose Link Structure=@n75, Alpha:float=@n68) -> (Pose:Component Space Pose Link Structure) {
    is=title FortAnimGraphNode_Flap
    bUseClamp=true
    targetClampMin=-10
    targetComponent=RotationZ
    alphaBoolBlend {blendOption=Linear}
    sourceDrives [
     {
      bUseRange=false
      drivingCurve=null
      multiplier=-0.2
      rangeMax=1
      rangeMin=-1
      remappedMax=1
      remappedMin=0
      sourceComponent=RotationZ
      useQuaternion=true
      sourceBone {boneName=thigh_r}
     }
     {
      bUseRange=false
      drivingCurve=null
      multiplier=1
      rangeMax=1
      rangeMin=-1
      remappedMax=1
      remappedMin=0
      sourceComponent=RotationZ
      useQuaternion=true
      sourceBone {boneName=thigh_r}
     }
    ]
    targetBone {boneName=dyn_apron_frnt_rb_04_r}
   }
  ]}
 EventGraph {nodes [
   Event BlueprintUpdateAnimation()
   n1 = Get Skel_Mesh()
   > Set leg_rot_r(leg_rot_r:double=GetAlphaFromReferenceTransformDelta(SkeletalMeshComponent=@n1, BoneName=thigh_twist_01_r, BaseName=pelvis, InputRange=(X=0,Y=-150), TwistAxis="0,0,1")) -> (Output_Get:double)
   > Set leg_rot_l(leg_rot_l:double=GetAlphaFromReferenceTransformDelta(SkeletalMeshComponent=@n1, BoneName=thigh_twist_01_l, BaseName=pelvis, InputRange=(X=0,Y=-150), TwistAxis="0,0,1")) -> (Output_Get:double)
   Event BlueprintInitializeAnimation()
   > Set Skel_Mesh(Skel_Mesh=GetOwnerSkeletalMeshComponent())
  ]}
]
variables [
 leg_rot_l {type=number}
 leg_rot_r {type=number}
 Skel_Mesh {type=SkeletalMeshComponent}
]
