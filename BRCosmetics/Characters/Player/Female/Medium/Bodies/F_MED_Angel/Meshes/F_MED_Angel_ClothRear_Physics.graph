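# Physics asset for the rear skirt cloth of the F_MED_Angel female player body.
# Left and right three-bone dyn_skirt_rear chains hang from small driver capsules, with swing and twist limits widening toward the hem.
# Cross constraints tie the two chains together at their second and third bones, keeping the panels moving together.
# Thigh and calf capsules add leg volumes, joined at the knees by constraints with free swing and twist.
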PhysicsAsset /BRCosmetics/Characters/Player/Female/Medium/Bodies/F_MED_Angel/Meshes/F_MED_Angel_ClothRear_Physics
bodies {
 calf_l [
  {shapeType=Capsule, length=10.6372, radius=7.7346, center=<-17.9071 -6.3616 -3.2398>, rotation=<rot -2.3696 -75.1145 -129.3284>}
  {shapeType=Capsule, length=28, radius=6.75, center=<-37.5277 -3.9692 -0.878>, rotation=<rot 90 -1.7899 0.5402>}
 ]
 calf_r [
  {shapeType=Capsule, length=10.6372, radius=7.7346, center=<19.4517 6.8568 1.2642>, rotation=<rot -2.3696 -75.1144 -129.3282>}
  {shapeType=Capsule, length=25.5, radius=6.75, center=<38.535 3.2624 0.7113>, rotation=<rot -90 0 -6.1095>}
 ]
 driver_skirt_rear_01_l [{shapeType=Capsule, length=0.505, radius=0.505, rotation=<rot 0 -90 0>}]
 driver_skirt_rear_01_r [{shapeType=Capsule, length=0.505, radius=0.505, rotation=<rot 0 -90 0>}]
 dyn_skirt_rear_01_l [{shapeType=Capsule, length=12, radius=4.75, center=<3 0 0>, rotation=<rot 90 0 0>}]
 dyn_skirt_rear_01_r [{shapeType=Capsule, length=12, radius=4.75, center=<3 0 0>, rotation=<rot 90 0 0>}]
 dyn_skirt_rear_02_l [{shapeType=Capsule, length=13.25, radius=5.75, center=<4 0 0>, rotation=<rot 90 0 0>}]
 dyn_skirt_rear_02_r [{shapeType=Capsule, length=13.25, radius=5.75, center=<4 0 0>, rotation=<rot 90 0 0>}]
 dyn_skirt_rear_03_l [
  {shapeType=Capsule, length=13.5, radius=9.75, center=<5.1856 0 -3.679>, rotation=<rot 77.5183 0 0>}
  {shapeType=Capsule, length=14.5, radius=8.5, center=<26.6691 0 -7.2068>, rotation=<rot 74.9999 0 0>}
 ]
 dyn_skirt_rear_03_r [
  {shapeType=Capsule, length=14.25, radius=9, center=<5.4809 0.0001 -2.3452>, rotation=<rot 77.5186 0 0>}
  {shapeType=Capsule, length=13.5, radius=8.75, center=<27.032 0 -7.2435>, rotation=<rot 74.9999 0 0>}
 ]
 thigh_l [{shapeType=Capsule, length=15.2413, radius=5.7905, center=<-17.9212 2.1737 -0.551>, rotation=<rot -0.5472 -88.3726 -94.5579>}]
 thigh_r [{shapeType=Capsule, length=15.2413, radius=5.7905, center=<17.9212 -2.1737 -0.551>, rotation=<rot -0.5472 -88.3726 -94.5579>}]
}
constraints [
 {
  bone1Name=calf_l
  bone2Name=thigh_l
  swing1LimitDegrees=45
  swing1Motion=Free
  swing2LimitDegrees=45
  swing2Motion=Free
  twistLimitDegrees=45
  twistMotion=Free
 }
 {
  bone1Name=calf_r
  bone2Name=thigh_r
  swing1LimitDegrees=45
  swing1Motion=Free
  swing2LimitDegrees=45
  swing2Motion=Free
  twistLimitDegrees=45
  twistMotion=Free
 }
 {bone1Name=dyn_skirt_rear_01_r, bone2Name=driver_skirt_rear_01_r, swing1LimitDegrees=15, swing2LimitDegrees=25, twistLimitDegrees=2}
 {bone1Name=dyn_skirt_rear_02_r, bone2Name=dyn_skirt_rear_01_r, swing1LimitDegrees=20, swing2LimitDegrees=35, twistLimitDegrees=15}
 {bone1Name=dyn_skirt_rear_03_r, bone2Name=dyn_skirt_rear_02_r, swing1LimitDegrees=45, swing2LimitDegrees=45, twistLimitDegrees=30}
 {bone1Name=dyn_skirt_rear_01_l, bone2Name=driver_skirt_rear_01_l, swing1LimitDegrees=15, swing2LimitDegrees=25, twistLimitDegrees=2}
 {bone1Name=dyn_skirt_rear_02_l, bone2Name=dyn_skirt_rear_01_l, swing1LimitDegrees=20, swing2LimitDegrees=35, twistLimitDegrees=15}
 {bone1Name=dyn_skirt_rear_03_l, bone2Name=dyn_skirt_rear_02_l, swing1LimitDegrees=45, swing2LimitDegrees=45, twistLimitDegrees=30}
 {bone1Name=dyn_skirt_rear_03_l, bone2Name=dyn_skirt_rear_03_r, swing1LimitDegrees=20, swing2LimitDegrees=20, twistLimitDegrees=20}
 {bone1Name=dyn_skirt_rear_02_l, bone2Name=dyn_skirt_rear_02_r, swing1LimitDegrees=20, swing2LimitDegrees=20, twistLimitDegrees=20}
]
