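# Physics asset for the side cloth skirt panels of the female medium Angel player body.
# Each leg has a skirt master bone jointed to the thigh, feeding three three-bone capsule chains (side, sideMid, frontMid), alongside thigh and calf capsules.
# Chain links use limited swing and tight twist, while master-to-thigh and calf-to-thigh joints rotate freely.
# Setup is asymmetric: right chain roots rotate freely on their master, while left roots are limited.
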PhysicsAsset /BRCosmetics/Characters/Player/Female/Medium/Bodies/F_MED_Angel/Meshes/F_MED_Angel_ClothSides_Physics
bodies {
 calf_l [{shapeType=Capsule, length=10.6372, radius=7.7346, center=<-17.9071 -6.3616 -3.2398>, rotation=<rot -2.3696 -75.1145 -129.3284>}]
 calf_r [{shapeType=Capsule, length=10.6372, radius=7.7346, center=<19.4517 6.8568 1.2642>, rotation=<rot -2.3696 -75.1144 -129.3282>}]
 dyn_skirt_frontMid_01_l [{shapeType=Capsule, length=1, radius=1}]
 dyn_skirt_frontMid_01_r [{shapeType=Capsule, length=1, radius=1}]
 dyn_skirt_frontMid_02_l [{shapeType=Capsule, length=9.75, radius=3, center=<5 0 -1.501>, rotation=<rot 87.4742 -0.0004 -0.0004>}]
 dyn_skirt_frontMid_02_r [{shapeType=Capsule, length=9.75, radius=3, center=<5 0 -1.501>, rotation=<rot 87.4742 -0.0004 -0.0004>}]
 dyn_skirt_frontMid_03_l [{shapeType=Capsule, length=14.75, radius=4.75, center=<11.9165 0 -3.197>, rotation=<rot 90 0 0>}]
 dyn_skirt_frontMid_03_r [{shapeType=Capsule, length=14.75, radius=4.75, center=<11.9165 0 -3.197>, rotation=<rot 90 0 0>}]
 dyn_skirt_master_l [{shapeType=Capsule, length=0.505, radius=0.505, rotation=<rot 0 -90 0>}]
 dyn_skirt_master_r [{shapeType=Capsule, length=0.505, radius=0.505, rotation=<rot 0 -90 0>}]
 dyn_skirt_sideMid_01_l [{shapeType=Capsule, length=1, radius=1}]
 dyn_skirt_sideMid_01_r [{shapeType=Capsule, length=1, radius=1}]
 dyn_skirt_sideMid_02_l [{shapeType=Capsule, length=8, radius=5, center=<4 0 0>, rotation=<rot 90 -0.0149 -0.015>}]
 dyn_skirt_sideMid_02_r [{shapeType=Capsule, length=8, radius=5, center=<4 0 0>, rotation=<rot 90 -0.0149 -0.015>}]
 dyn_skirt_sideMid_03_l [{shapeType=Capsule, length=15.5, radius=4.25, center=<13.0147 0 -0.0001>, rotation=<rot 90 -0.0298 -0.03>}]
 dyn_skirt_sideMid_03_r [{shapeType=Capsule, length=15.5, radius=4.25, center=<13.0147 0 -0.0001>, rotation=<rot 90 -0.0298 -0.03>}]
 dyn_skirt_side_01_l [{shapeType=Capsule, length=1, radius=1}]
 dyn_skirt_side_01_r [{shapeType=Capsule, length=1, radius=1}]
 dyn_skirt_side_02_l [{shapeType=Capsule, length=13.75, radius=5, center=<4 0 -2.5598>, rotation=<rot 90 178.7341 178.7342>}]
 dyn_skirt_side_02_r [{shapeType=Capsule, length=13.75, radius=5, center=<4 0 -2.5598>, rotation=<rot 90 178.7341 178.7342>}]
 dyn_skirt_side_03_l [{shapeType=Capsule, length=13.75, radius=5, center=<13.4522 0 -2.7164>, rotation=<rot 90 -2.5305 -2.5304>}]
 dyn_skirt_side_03_r [{shapeType=Capsule, length=13.75, radius=5, center=<13.4522 0 -2.7164>, rotation=<rot 90 -2.5305 -2.5304>}]
 thigh_l [{shapeType=Capsule, length=15.2413, radius=5.7905, center=<-17.9212 2.1737 -0.551>, rotation=<rot -0.5472 -88.3726 -94.5579>}]
 thigh_r [{shapeType=Capsule, length=15.2413, radius=5.7905, center=<17.9212 -2.1737 -0.551>, rotation=<rot -0.5472 -88.3726 -94.5579>}]
}
constraints [
 {bone1Name=dyn_skirt_side_01_l, bone2Name=dyn_skirt_master_l, swing1LimitDegrees=25, swing2LimitDegrees=25, twistLimitDegrees=2}
 {bone1Name=dyn_skirt_side_02_l, bone2Name=dyn_skirt_side_01_l, swing1LimitDegrees=30, swing2LimitDegrees=25, twistLimitDegrees=2}
 {bone1Name=dyn_skirt_side_03_l, bone2Name=dyn_skirt_side_02_l, swing1LimitDegrees=30, swing2LimitDegrees=25, twistLimitDegrees=2}
 {bone1Name=dyn_skirt_sideMid_01_l, bone2Name=dyn_skirt_master_l, swing1LimitDegrees=25, swing2LimitDegrees=20, twistLimitDegrees=2}
 {bone1Name=dyn_skirt_sideMid_02_l, bone2Name=dyn_skirt_sideMid_01_l, swing1LimitDegrees=30, swing2LimitDegrees=25, twistLimitDegrees=2}
 {bone1Name=dyn_skirt_sideMid_03_l, bone2Name=dyn_skirt_sideMid_02_l, swing1LimitDegrees=30, swing2LimitDegrees=25, twistLimitDegrees=2}
 {bone1Name=dyn_skirt_frontMid_01_l, bone2Name=dyn_skirt_master_l, swing1LimitDegrees=25, swing2LimitDegrees=20, twistLimitDegrees=2}
 {bone1Name=dyn_skirt_frontMid_02_l, bone2Name=dyn_skirt_frontMid_01_l, swing1LimitDegrees=30, swing2LimitDegrees=25, twistLimitDegrees=2}
 {bone1Name=dyn_skirt_frontMid_03_l, bone2Name=dyn_skirt_frontMid_02_l, swing1LimitDegrees=30, swing2LimitDegrees=25, twistLimitDegrees=2}
 {
  bone1Name=dyn_skirt_master_l
  bone2Name=thigh_l
  swing1LimitDegrees=45
  swing1Motion=Free
  swing2LimitDegrees=45
  swing2Motion=Free
  twistLimitDegrees=45
  twistMotion=Free
 }
 {
  bone1Name=calf_l
  bone2Name=thigh_l
  swing1LimitDegrees=45
  swing1Motion=Free
  swing2LimitDegrees=45
  swing2Motion=Free
  twistLimitDegrees=45
  twistMotion=Free
 }
 {
  bone1Name=dyn_skirt_side_01_r
  bone2Name=dyn_skirt_master_r
  swing1LimitDegrees=45
  swing1Motion=Free
  swing2LimitDegrees=45
  swing2Motion=Free
  twistLimitDegrees=45
  twistMotion=Free
 }
 {
  bone1Name=dyn_skirt_sideMid_01_r
  bone2Name=dyn_skirt_master_r
  swing1LimitDegrees=45
  swing1Motion=Free
  swing2LimitDegrees=45
  swing2Motion=Free
  twistLimitDegrees=45
  twistMotion=Free
 }
 {
  bone1Name=dyn_skirt_frontMid_01_r
  bone2Name=dyn_skirt_master_r
  swing1LimitDegrees=45
  swing1Motion=Free
  swing2LimitDegrees=45
  swing2Motion=Free
  twistLimitDegrees=45
  twistMotion=Free
 }
 {bone1Name=dyn_skirt_side_02_r, bone2Name=dyn_skirt_side_01_r, swing1LimitDegrees=30, swing2LimitDegrees=25, twistLimitDegrees=2}
 {bone1Name=dyn_skirt_side_03_r, bone2Name=dyn_skirt_side_02_r, swing1LimitDegrees=30, swing2LimitDegrees=25, twistLimitDegrees=2}
 {bone1Name=dyn_skirt_sideMid_02_r, bone2Name=dyn_skirt_sideMid_01_r, swing1LimitDegrees=30, swing2LimitDegrees=25, twistLimitDegrees=2}
 {bone1Name=dyn_skirt_sideMid_03_r, bone2Name=dyn_skirt_sideMid_02_r, swing1LimitDegrees=30, swing2LimitDegrees=25, twistLimitDegrees=2}
 {bone1Name=dyn_skirt_frontMid_02_r, bone2Name=dyn_skirt_frontMid_01_r, swing1LimitDegrees=30, swing2LimitDegrees=25, twistLimitDegrees=2}
 {bone1Name=dyn_skirt_frontMid_03_r, bone2Name=dyn_skirt_frontMid_02_r, swing1LimitDegrees=30, swing2LimitDegrees=25, twistLimitDegrees=2}
 {
  bone1Name=dyn_skirt_master_r
  bone2Name=thigh_r
  swing1LimitDegrees=45
  swing1Motion=Free
  swing2LimitDegrees=45
  swing2Motion=Free
  twistLimitDegrees=45
  twistMotion=Free
 }
 {
  bone1Name=calf_r
  bone2Name=thigh_r
  swing1LimitDegrees=45
  swing1Motion=Free
  swing2LimitDegrees=45
  swing2Motion=Free
  twistLimitDegrees=45
  twistMotion=Free
 }
]
