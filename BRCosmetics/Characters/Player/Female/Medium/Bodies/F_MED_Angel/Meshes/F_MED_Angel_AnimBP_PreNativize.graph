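# Post-process-style AnimBP for the F_MED_Angel cosmetic body, adding secondary motion to its cloth, skirt and feathers.
# It copies the parent mesh pose, applies AnimDynamics calf feathers and corrective bone offsets, then caches it.
# When rigid bodies are enabled, four chained RigidBody simulations and further offsets layer on top.
# EventGraph sets alphas from speed, crouch, thigh twist, emotes and long capes, and gravity via a deprecated helper; the skydiving branch is unconnected.
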AnimBlueprint /BRCosmetics/Characters/Player/Female/Medium/Bodies/F_MED_Angel/Meshes/F_MED_Angel_AnimBP_PreNativize
target_skeleton=&/F_MED_Angel_Skeleton
graphs [
 AnimGraph {nodes [
   n1 = Get crouch_skirt_alpha() -> (crouch_skirt_alpha:double)
   "Transform (Modify) Bone - Bone: dyn_parent_cloth_front"(Alpha=@n1, Translation="0,0,0", Rotation="5,0.000,3", Scale="1,1,1") {
    rotationMode=BMM_Additive
    rotationSpace=BCS_BoneSpace
    translationMode=BMM_Additive
    translationSpace=BCS_BoneSpace
    alphaBoolBlend {blendOption=Linear}
    alphaScaleBiasClamp {bInterpResult=true, interpSpeedDecreasing=5, interpSpeedIncreasing=15}
    boneToModify {boneName=dyn_parent_cloth_front}
    rotation {pitch=5, roll=3}
   }
   n2 = RigidBody(ComponentPose=Local To Component(LocalPose=Use cached pose 'AnimCache'()), Alpha=Get side_cloth_alpha() -> (side_cloth_alpha:double)) {
    bOverrideWorldGravity=true
    overridePhysicsAsset=&/F_MED_Angel_ClothSides_Physics
    simulationSpace=BaseBoneSpace
    alphaBoolBlend {blendOption=Linear}
    alphaScaleBiasClamp {bInterpResult=true, interpSpeedDecreasing=1, interpSpeedIncreasing=1}
    baseBoneRef {boneName=pelvis}
    componentAppliedLinearAccClamp=<1100 1100 1100>
    componentLinearAccScale=<0.7 0.7 0.5>
    componentLinearVelScale=<0.7 0.7 0.5>
    overrideWorldGravity {z=-980}
   }
   n3 = RigidBody!(
    ComponentPose=@n2
    Alpha=Get cloth_alpha() -> (cloth_alpha:double)
    OverrideWorldGravity:Vector=Get gravity_override() -> (gravity_override:Vector)
    ExternalForce
   ) -> (Pose) {
    bOverrideWorldGravity=true
    overridePhysicsAsset=&/F_MED_Angel_ClothRear_Physics
    simulationSpace=BaseBoneSpace
    alphaBoolBlend {blendOption=Linear}
    alphaScaleBiasClamp {bInterpResult=true, interpSpeedDecreasing=1, interpSpeedIncreasing=1}
    baseBoneRef {boneName=root}
    componentAppliedLinearAccClamp=<1100 1100 1100>
    componentLinearAccScale=<0.7 0.7 0.75>
    componentLinearVelScale=<0.7 0.7 0.75>
    overrideWorldGravity {z=-920}
   }
   n4 = RigidBody(ComponentPose=@n3, OverrideWorldGravity="0,0,-980") {
    bOverrideWorldGravity=true
    overridePhysicsAsset=&/F_MED_Angel_Physics
    simulationSpace=BaseBoneSpace
    alphaBoolBlend {blendOption=Linear}
    baseBoneRef {boneName=root}
    componentAppliedLinearAccClamp=<1100 1100 1100>
    componentLinearAccScale=<1 0.7 0.7>
    componentLinearVelScale=<1 0.7 0.7>
    overrideWorldGravity {z=-900}
   }
   n5 = RigidBody!(ComponentPose=@n4, Alpha, ExternalForce) -> (Pose) {
    bClampLinearTranslationLimitToRefPose=true
    bOverrideWorldGravity=true
    overridePhysicsAsset=&/F_MED_Angel_PadFeathers_Physics
    simulationSpace=BaseBoneSpace
    alphaBoolBlend {blendOption=Linear}
    baseBoneRef {boneName=spine_05}
    componentAppliedLinearAccClamp=<500 500 500>
    componentLinearAccScale=<1 1 1>
    componentLinearVelScale=<1 1 1>
    overrideWorldGravity {z=-980}
   }
   n6 = "Transform (Modify) Bone - Bone: dyn_skirt_frontMid_01_l"(ComponentPose=@n5, Alpha=@n1, Translation="0,0,-2.5", Rotation="15,0.000,-35", Scale="1,1,1") {
    rotationMode=BMM_Additive
    rotationSpace=BCS_BoneSpace
    translationMode=BMM_Additive
    translationSpace=BCS_BoneSpace
    alphaBoolBlend {blendOption=Linear}
    alphaScaleBiasClamp {bInterpResult=true, interpSpeedDecreasing=5, interpSpeedIncreasing=15}
    boneToModify {boneName=dyn_skirt_frontMid_01_l}
    rotation {pitch=15, roll=-35}
    translation {z=-2.5}
   }
   n7 = "Transform (Modify) Bone - Bone: dyn_skirt_frontMid_02_l"(ComponentPose=@n6, Alpha=@n1, Translation="0,0,0", Rotation="50,0.000,0.000", Scale="1,1,1") {
    rotationMode=BMM_Additive
    rotationSpace=BCS_BoneSpace
    translationMode=BMM_Additive
    translationSpace=BCS_BoneSpace
    alphaBoolBlend {blendOption=Linear}
    alphaScaleBiasClamp {bInterpResult=true, interpSpeedDecreasing=5, interpSpeedIncreasing=15}
    boneToModify {boneName=dyn_skirt_frontMid_02_l}
    rotation {pitch=50}
   }
   n8 = "Transform (Modify) Bone - Bone: dyn_skirt_master_l"(
    ComponentPose=@n7
    Alpha=Get emote_shaolin_sip_active() -> (emote_shaolin_sip_active:double)
    Translation="0.674,3.589,-1.372"
    Rotation="0,0,0"
    Scale="1,1,1"
   ) {
    rotationMode=BMM_Additive
    rotationSpace=BCS_BoneSpace
    translationMode=BMM_Additive
    translationSpace=BCS_BoneSpace
    alphaBoolBlend {blendOption=Linear}
    alphaScaleBiasClamp {bInterpResult=true}
    boneToModify {boneName=dyn_skirt_master_l}
    translation=<0.674 3.589 -1.3724>
   }
   n9 = Get front_end_rear_alpha() -> (front_end_rear_alpha:double)
   n10 = "Transform (Modify) Bone - Bone: driver_skirt_rear_01_l"(ComponentPose=@n8, Alpha=@n9, Translation="0,0,0", Rotation="0,-2.5,0", Scale="1,1,1") {
    rotationMode=BMM_Additive
    rotationSpace=BCS_BoneSpace
    translationMode=BMM_Additive
    translationSpace=BCS_BoneSpace
    alphaBoolBlend {blendOption=Linear}
    boneToModify {boneName=driver_skirt_rear_01_l}
    rotation {yaw=-2.5}
   }
   n11 = "Transform (Modify) Bone - Bone: driver_skirt_rear_01_r"(ComponentPose=@n10, Alpha=@n9, Translation="0,0,0", Rotation="0,-2.5,0", Scale="1,1,1") {
    rotationMode=BMM_Additive
    rotationSpace=BCS_BoneSpace
    translationMode=BMM_Additive
    translationSpace=BCS_BoneSpace
    alphaBoolBlend {blendOption=Linear}
    boneToModify {boneName=driver_skirt_rear_01_r}
    rotation {yaw=-2.5}
   }
   Output Pose(Result=Blend Poses by bool(BlendPose_0=Component To Local(ComponentPose=@n11), BlendPose_1=Use cached pose 'AnimCache'(), bActiveValue=Get bRigidBodyEnabled()) {blendType=Linear}) {layerGroup=None}
   n12 = Get feather_alpha() -> (feather_alpha:double)
   n13 = "Constraint - dyn_skirt_frontMid_01_l (P)"(
    ComponentPose=Local To Component(LocalPose=Copy Pose From Mesh() {bCopyCurves=true, bUseAttachedParent=true})
    Alpha=1
    ConstraintWeights_0=1
   ) {
    alphaBoolBlend {blendOption=Linear}
    boneToModify {boneName=dyn_skirt_frontMid_01_l}
    constraintSetup [{targetBone {boneName=thigh_l}}]
    constraintWeights [1]
   }
   n14 = Anim Dynamics - Bone: dyn_calf_fthr_01_l(ComponentPose=@n13, Alpha=0.3, ExternalForce="0,0,0") {
    angularDampingOverride=0.985
    angularSpringConstant=500
    bAngularSpring=true
    bGravityOverrideInSimSpace=true
    bLinearSpring=true
    bOverrideAngularDamping=true
    bOverrideLinearDamping=true
    gravityScale=0.085
    linearDampingOverride=0.99
    linearSpringConstant=600
    alphaBoolBlend {blendOption=Linear}
    boundBone {boneName=dyn_calf_fthr_01_l}
    componentAppliedLinearAccClamp=<1000 1000 1000>
    componentLinearVelScale=<100 100 100>
    physicsBodyDefinitions [{
      sphereCollisionRadius=0
      boundBone {boneName=dyn_calf_fthr_01_l}
      boxExtents=<30 30 30>
      constraintSetup {
       angularTargetAxis=AxisY
       angularLimitsMax=<0 0 0>
       angularLimitsMin=<0 0 0>
       angularTarget=<0 0 0>
       linearAxesMax=<2 2 1>
       linearAxesMin=<-2 -2 -1>
      }
      localJointOffset=<0 0 0>
     }]
   }
   n15 = Anim Dynamics - Bone: dyn_calf_fthr_02_l(ComponentPose=@n14, Alpha=0.35, ExternalForce="0,0,0") {
    angularDampingOverride=0.985
    angularSpringConstant=300
    bAngularSpring=true
    bGravityOverrideInSimSpace=true
    bLinearSpring=true
    bOverrideAngularDamping=true
    bOverrideLinearDamping=true
    gravityScale=0.15
    linearDampingOverride=0.99
    linearSpringConstant=200
    alphaBoolBlend {blendOption=Linear}
    boundBone {boneName=dyn_calf_fthr_02_l}
    componentAppliedLinearAccClamp=<1000 1000 1000>
    componentLinearVelScale=<100 100 100>
    physicsBodyDefinitions [{
      sphereCollisionRadius=0
      boundBone {boneName=dyn_calf_fthr_02_l}
      boxExtents=<30 45 7>
      constraintSetup {
       angularTargetAxis=AxisY
       angularLimitsMax=<0 0 0>
       angularLimitsMin=<0 0 0>
       angularTarget=<0 0 0>
       linearAxesMax=<2 2 1>
       linearAxesMin=<-2 -2 -1>
      }
      localJointOffset=<0 0 0>
     }]
   }
   n16 = Anim Dynamics - Bone: dyn_calf_fthr_03_l(ComponentPose=@n15, Alpha=0.4, ExternalForce="0,0,0") {
    angularDampingOverride=0.985
    angularSpringConstant=500
    bAngularSpring=true
    bGravityOverrideInSimSpace=true
    bLinearSpring=true
    bOverrideAngularDamping=true
    bOverrideLinearDamping=true
    gravityScale=0.3
    linearDampingOverride=0.99
    linearSpringConstant=500
    alphaBoolBlend {blendOption=Linear}
    boundBone {boneName=dyn_calf_fthr_03_l}
    componentAppliedLinearAccClamp=<1000 1000 1000>
    componentLinearVelScale=<100 100 100>
    physicsBodyDefinitions [{
      sphereCollisionRadius=0
      boundBone {boneName=dyn_calf_fthr_03_l}
      boxExtents=<20 10 15>
      constraintSetup {
       angularTargetAxis=AxisY
       angularLimitsMax=<0 0 0>
       angularLimitsMin=<0 0 0>
       angularTarget=<0 0 0>
       linearAxesMax=<2 2 1>
       linearAxesMin=<-2 -2 -1>
      }
      localJointOffset=<0 0 0>
     }]
   }
   n17 = Anim Dynamics - Bone: dyn_calf_fthr_04_l(ComponentPose=@n16, Alpha=0.4, ExternalForce="0,0,0") {
    angularDampingOverride=0.985
    angularSpringConstant=500
    bAngularSpring=true
    bGravityOverrideInSimSpace=true
    bLinearSpring=true
    bOverrideAngularDamping=true
    bOverrideLinearDamping=true
    gravityScale=0.085
    linearDampingOverride=0.99
    linearSpringConstant=600
    alphaBoolBlend {blendOption=Linear}
    boundBone {boneName=dyn_calf_fthr_04_l}
    componentAppliedLinearAccClamp=<1000 1000 1000>
    componentLinearVelScale=<100 100 100>
    physicsBodyDefinitions [{
      sphereCollisionRadius=0
      boundBone {boneName=dyn_calf_fthr_04_l}
      boxExtents=<30 30 30>
      constraintSetup {
       angularTargetAxis=AxisY
       angularLimitsMax=<0 0 0>
       angularLimitsMin=<0 0 0>
       angularTarget=<0 0 0>
       linearAxesMax=<2 2 1>
       linearAxesMin=<-2 -2 -1>
      }
      localJointOffset=<0 0 0>
     }]
   }
   n18 = Anim Dynamics - Bone: dyn_calf_fthr_05_l(ComponentPose=@n17, Alpha=0.32, ExternalForce="0,0,0") {
    angularDampingOverride=0.985
    angularSpringConstant=300
    bAngularSpring=true
    bGravityOverrideInSimSpace=true
    bLinearSpring=true
    bOverrideAngularDamping=true
    bOverrideLinearDamping=true
    gravityScale=0.15
    linearDampingOverride=0.99
    linearSpringConstant=200
    alphaBoolBlend {blendOption=Linear}
    boundBone {boneName=dyn_calf_fthr_05_l}
    componentAppliedLinearAccClamp=<1000 1000 1000>
    componentLinearVelScale=<100 100 100>
    physicsBodyDefinitions [{
      sphereCollisionRadius=0
      boundBone {boneName=dyn_calf_fthr_05_l}
      boxExtents=<30 45 7>
      constraintSetup {
       angularTargetAxis=AxisY
       angularLimitsMax=<0 0 0>
       angularLimitsMin=<0 0 0>
       angularTarget=<0 0 0>
       linearAxesMax=<1.5 1.5 0.5>
       linearAxesMin=<-1.5 -1.5 -0.5>
      }
      localJointOffset=<0 0 0>
     }]
   }
   n19 = Anim Dynamics - Bone: dyn_calf_fthr_01_r(ComponentPose=@n18, Alpha=0.36, ExternalForce="0,0,0") {
    angularDampingOverride=0.985
    angularSpringConstant=500
    bAngularSpring=true
    bGravityOverrideInSimSpace=true
    bLinearSpring=true
    bOverrideAngularDamping=true
    bOverrideLinearDamping=true
    gravityScale=0.3
    linearDampingOverride=0.99
    linearSpringConstant=500
    alphaBoolBlend {blendOption=Linear}
    boundBone {boneName=dyn_calf_fthr_01_r}
    componentAppliedLinearAccClamp=<1000 1000 1000>
    componentLinearVelScale=<100 100 100>
    physicsBodyDefinitions [{
      sphereCollisionRadius=0
      boundBone {boneName=dyn_calf_fthr_01_r}
      boxExtents=<20 10 15>
      constraintSetup {
       angularTargetAxis=AxisY
       angularLimitsMax=<0 0 0>
       angularLimitsMin=<0 0 0>
       angularTarget=<0 0 0>
       linearAxesMax=<2 2 1>
       linearAxesMin=<-2 -2 -1>
      }
      localJointOffset=<0 0 0>
     }]
   }
   n20 = Anim Dynamics - Bone: dyn_calf_fthr_02_r(ComponentPose=@n19, Alpha=0.4, ExternalForce="0,0,0") {
    angularDampingOverride=0.985
    angularSpringConstant=500
    bAngularSpring=true
    bGravityOverrideInSimSpace=true
    bLinearSpring=true
    bOverrideAngularDamping=true
    bOverrideLinearDamping=true
    gravityScale=0.085
    linearDampingOverride=0.99
    linearSpringConstant=600
    alphaBoolBlend {blendOption=Linear}
    boundBone {boneName=dyn_calf_fthr_02_r}
    componentAppliedLinearAccClamp=<1000 1000 1000>
    componentLinearVelScale=<100 100 100>
    physicsBodyDefinitions [{
      sphereCollisionRadius=0
      boundBone {boneName=dyn_calf_fthr_02_r}
      boxExtents=<30 30 30>
      constraintSetup {
       angularTargetAxis=AxisY
       angularLimitsMax=<0 0 0>
       angularLimitsMin=<0 0 0>
       angularTarget=<0 0 0>
       linearAxesMax=<2 2 1>
       linearAxesMin=<-2 -2 -1>
      }
      localJointOffset=<0 0 0>
     }]
   }
   n21 = Anim Dynamics - Bone: dyn_calf_fthr_03_r(ComponentPose=@n20, Alpha=0.45, ExternalForce="0,0,0") {
    angularDampingOverride=0.985
    angularSpringConstant=300
    bAngularSpring=true
    bGravityOverrideInSimSpace=true
    bLinearSpring=true
    bOverrideAngularDamping=true
    bOverrideLinearDamping=true
    gravityScale=0.15
    linearDampingOverride=0.99
    linearSpringConstant=200
    alphaBoolBlend {blendOption=Linear}
    boundBone {boneName=dyn_calf_fthr_03_r}
    componentAppliedLinearAccClamp=<1000 1000 1000>
    componentLinearVelScale=<100 100 100>
    physicsBodyDefinitions [{
      sphereCollisionRadius=0
      boundBone {boneName=dyn_calf_fthr_03_r}
      boxExtents=<30 45 7>
      constraintSetup {
       angularTargetAxis=AxisY
       angularLimitsMax=<0 0 0>
       angularLimitsMin=<0 0 0>
       angularTarget=<0 0 0>
       linearAxesMax=<2 2 1>
       linearAxesMin=<-2 -2 -1>
      }
      localJointOffset=<0 0 0>
     }]
   }
   n22 = Anim Dynamics - Bone: dyn_calf_fthr_04_r(ComponentPose=@n21, Alpha=0.3, ExternalForce="0,0,0") {
    angularDampingOverride=0.985
    angularSpringConstant=500
    bAngularSpring=true
    bGravityOverrideInSimSpace=true
    bLinearSpring=true
    bOverrideAngularDamping=true
    bOverrideLinearDamping=true
    gravityScale=0.3
    linearDampingOverride=0.99
    linearSpringConstant=500
    alphaBoolBlend {blendOption=Linear}
    boundBone {boneName=dyn_calf_fthr_04_r}
    componentAppliedLinearAccClamp=<1000 1000 1000>
    componentLinearVelScale=<100 100 100>
    physicsBodyDefinitions [{
      sphereCollisionRadius=0
      boundBone {boneName=dyn_calf_fthr_04_r}
      boxExtents=<20 10 15>
      constraintSetup {
       angularTargetAxis=AxisY
       angularLimitsMax=<0 0 0>
       angularLimitsMin=<0 0 0>
       angularTarget=<0 0 0>
       linearAxesMax=<2 2 1>
       linearAxesMin=<-2 -2 -1>
      }
      localJointOffset=<0 0 0>
     }]
   }
   n23 = Anim Dynamics - Bone: dyn_calf_fthr_05_r(ComponentPose=@n22, Alpha=0.32, ExternalForce="0,0,0") {
    angularDampingOverride=0.985
    angularSpringConstant=300
    bAngularSpring=true
    bGravityOverrideInSimSpace=true
    bLinearSpring=true
    bOverrideAngularDamping=true
    bOverrideLinearDamping=true
    gravityScale=0.15
    linearDampingOverride=0.99
    linearSpringConstant=200
    alphaBoolBlend {blendOption=Linear}
    boundBone {boneName=dyn_calf_fthr_05_r}
    componentAppliedLinearAccClamp=<1000 1000 1000>
    componentLinearVelScale=<100 100 100>
    physicsBodyDefinitions [{
      sphereCollisionRadius=0
      boundBone {boneName=dyn_calf_fthr_05_r}
      boxExtents=<30 45 7>
      constraintSetup {
       angularTargetAxis=AxisY
       angularLimitsMax=<0 0 0>
       angularLimitsMin=<0 0 0>
       angularTarget=<0 0 0>
       linearAxesMax=<1.5 1.5 0.5>
       linearAxesMin=<-1.5 -1.5 -0.5>
      }
      localJointOffset=<0 0 0>
     }]
   }
   n24 = "Transform (Modify) Bone - Bone: dyn_calf_base_l"(ComponentPose=@n23, Alpha=@n12, Translation="0,0,0", Rotation="-20,0.000,0.000", Scale="1,1,1") {
    rotationMode=BMM_Additive
    rotationSpace=BCS_BoneSpace
    translationSpace=BCS_BoneSpace
    alphaBoolBlend {blendOption=Linear}
    alphaScaleBiasClamp {bInterpResult=true, interpSpeedDecreasing=9, interpSpeedIncreasing=9}
    boneToModify {boneName=dyn_calf_base_l}
    rotation {pitch=-20}
   }
   n25 = "Transform (Modify) Bone - Bone: dyn_calf_base_r"(ComponentPose=@n24, Alpha=@n12, Translation="0,0,0", Rotation="20,0.000,0.000", Scale="1,1,1") {
    rotationMode=BMM_Additive
    rotationSpace=BCS_BoneSpace
    translationSpace=BCS_BoneSpace
    alphaBoolBlend {blendOption=Linear}
    alphaScaleBiasClamp {bInterpResult=true, interpSpeedDecreasing=9, interpSpeedIncreasing=9}
    boneToModify {boneName=dyn_calf_base_r}
    rotation {pitch=20}
   }
   n26 = Get crouch_skirt_alpha()
   n27 = "Transform (Modify) Bone - Bone: dyn_skirt_sideMid_01_r"(ComponentPose=@n25, Alpha=@n26, Translation="0,0,0", Rotation="-30,0.000,0.000", Scale="1,1,1") {
    rotationMode=BMM_Additive
    rotationSpace=BCS_BoneSpace
    translationSpace=BCS_BoneSpace
    alphaBoolBlend {blendOption=Linear}
    alphaScaleBiasClamp {bInterpResult=true, interpSpeedDecreasing=9, interpSpeedIncreasing=9}
    boneToModify {boneName=dyn_skirt_sideMid_01_r}
    rotation {pitch=-30}
   }
   n28 = "Transform (Modify) Bone - Bone: dyn_skirt_sideMid_02_r"(ComponentPose=@n27, Alpha=@n26, Translation="0,0,0", Rotation="10,0.000,0.000", Scale="1,1,1") {
    rotationMode=BMM_Additive
    rotationSpace=BCS_BoneSpace
    translationSpace=BCS_BoneSpace
    alphaBoolBlend {blendOption=Linear}
    alphaScaleBiasClamp {bInterpResult=true, interpSpeedDecreasing=9, interpSpeedIncreasing=9}
    boneToModify {boneName=dyn_skirt_sideMid_02_r}
    rotation {pitch=10}
   }
   n29 = "Transform (Modify) Bone - Bone: dyn_skirt_sideMid_03_r"(ComponentPose=@n28, Alpha=@n26, Translation="0,0,0", Rotation="10,0,0.000", Scale="1,1,1") {
    rotationMode=BMM_Additive
    rotationSpace=BCS_BoneSpace
    translationSpace=BCS_BoneSpace
    alphaBoolBlend {blendOption=Linear}
    alphaScaleBiasClamp {bInterpResult=true, interpSpeedDecreasing=9, interpSpeedIncreasing=9}
    boneToModify {boneName=dyn_skirt_sideMid_03_r}
    rotation {pitch=10}
   }
   n30 = Get crouch_skirt_alpha()
   n31 = "Transform (Modify) Bone - Bone: dyn_skirt_frontMid_01_l"(ComponentPose=@n29, Alpha=@n30, Translation="0,0,0", Rotation="-9.941,23.679,1.096", Scale="1,1,1") {
    rotationMode=BMM_Additive
    rotationSpace=BCS_BoneSpace
    translationSpace=BCS_BoneSpace
    alphaBoolBlend {blendOption=Linear}
    alphaScaleBiasClamp {bInterpResult=true, interpSpeedDecreasing=9, interpSpeedIncreasing=9}
    boneToModify {boneName=dyn_skirt_frontMid_01_l}
    rotation=<rot -9.941 23.679 1.096>
   }
   n32 = "Transform (Modify) Bone - Bone: dyn_skirt_frontMid_02_l"(ComponentPose=@n31, Alpha=@n30, Translation="0,0,0", Rotation="38.698,10.676,-11.091", Scale="1,1,1") {
    rotationMode=BMM_Additive
    rotationSpace=BCS_BoneSpace
    translationSpace=BCS_BoneSpace
    alphaBoolBlend {blendOption=Linear}
    alphaScaleBiasClamp {bInterpResult=true, interpSpeedDecreasing=9, interpSpeedIncreasing=9}
    boneToModify {boneName=dyn_skirt_frontMid_02_l}
    rotation=<rot 38.698 10.676 -11.091>
   }
   n33 = "Transform (Modify) Bone - Bone: dyn_skirt_frontMid_03_l"(ComponentPose=@n32, Alpha=@n30, Translation="0,0,0", Rotation="20.000,15.000,-25.510", Scale="1,1,1") {
    rotationMode=BMM_Additive
    rotationSpace=BCS_BoneSpace
    translationSpace=BCS_BoneSpace
    alphaBoolBlend {blendOption=Linear}
    alphaScaleBiasClamp {bInterpResult=true, interpSpeedDecreasing=9, interpSpeedIncreasing=9}
    boneToModify {boneName=dyn_skirt_frontMid_03_l}
    rotation=<rot 20 15 -25.51>
   }
   n34 = Get crouch_skirt_alpha()
   n35 = "Transform (Modify) Bone - Bone: dyn_skirt_side_01_l"(ComponentPose=@n33, Alpha=@n34, Translation="0,0,0", Rotation="-10,25,0", Scale="1,1,1") {
    rotationMode=BMM_Additive
    rotationSpace=BCS_BoneSpace
    translationSpace=BCS_BoneSpace
    alphaBoolBlend {blendOption=Linear}
    alphaScaleBiasClamp {bInterpResult=true, interpSpeedDecreasing=9, interpSpeedIncreasing=9}
    boneToModify {boneName=dyn_skirt_side_01_l}
    rotation {pitch=-10, yaw=25}
   }
   n36 = "Transform (Modify) Bone - Bone: dyn_skirt_side_02_l"(ComponentPose=@n35, Alpha=@n34, Translation="0,0,0", Rotation="-10.025,33.747,-21.708", Scale="1,1,1") {
    rotationMode=BMM_Additive
    rotationSpace=BCS_BoneSpace
    translationSpace=BCS_BoneSpace
    alphaBoolBlend {blendOption=Linear}
    alphaScaleBiasClamp {bInterpResult=true, interpSpeedDecreasing=9, interpSpeedIncreasing=9}
    boneToModify {boneName=dyn_skirt_side_02_l}
    rotation=<rot -10.025 33.747 -21.708>
   }
   n37 = "Transform (Modify) Bone - Bone: dyn_skirt_side_03_l"(ComponentPose=@n36, Alpha=@n34, Translation="0,0,0", Rotation="20,10,0.000", Scale="1,1,1") {
    rotationMode=BMM_Additive
    rotationSpace=BCS_BoneSpace
    translationSpace=BCS_BoneSpace
    alphaBoolBlend {blendOption=Linear}
    alphaScaleBiasClamp {bInterpResult=true, interpSpeedDecreasing=9, interpSpeedIncreasing=9}
    boneToModify {boneName=dyn_skirt_side_03_l}
    rotation {pitch=20, yaw=10}
   }
   n38 = "Transform (Modify) Bone - Bone: dyn_skirt_sideMid_01_l"(ComponentPose=@n37, Alpha=@n34, Translation="0,0,0", Rotation="-19.882,35.834,-1.899", Scale="1,1,1") {
    rotationMode=BMM_Additive
    rotationSpace=BCS_BoneSpace
    translationSpace=BCS_BoneSpace
    alphaBoolBlend {blendOption=Linear}
    alphaScaleBiasClamp {bInterpResult=true, interpSpeedDecreasing=9, interpSpeedIncreasing=9}
    boneToModify {boneName=dyn_skirt_sideMid_01_l}
    rotation=<rot -19.882 35.834 -1.899>
   }
   n39 = "Transform (Modify) Bone - Bone: dyn_skirt_sideMid_02_l"(ComponentPose=@n38, Alpha=@n34, Translation="0,0,0", Rotation="-4.683,-18.689,-0.007", Scale="1,1,1") {
    rotationMode=BMM_Additive
    rotationSpace=BCS_BoneSpace
    translationSpace=BCS_BoneSpace
    alphaBoolBlend {blendOption=Linear}
    alphaScaleBiasClamp {bInterpResult=true, interpSpeedDecreasing=9, interpSpeedIncreasing=9}
    boneToModify {boneName=dyn_skirt_sideMid_02_l}
    rotation=<rot -4.683 -18.689 -0.007>
   }
   n40 = "Transform (Modify) Bone - Bone: dyn_skirt_sideMid_03_l"(ComponentPose=@n39, Alpha=@n34, Translation="0,0,0", Rotation="-8.772,13.586,-18.200", Scale="1,1,1") {
    rotationMode=BMM_Additive
    rotationSpace=BCS_BoneSpace
    translationSpace=BCS_BoneSpace
    alphaBoolBlend {blendOption=Linear}
    alphaScaleBiasClamp {bInterpResult=true, interpSpeedDecreasing=9, interpSpeedIncreasing=9}
    boneToModify {boneName=dyn_skirt_sideMid_03_l}
    rotation=<rot -8.772 13.5861 -18.2001>
   }
   n41 = Get thigh_up_alpha() -> (thigh_up_alpha:double)
   n42 = "Transform (Modify) Bone - Bone: dyn_skirt_side_01_r"(ComponentPose=@n40, Alpha=@n41, Translation="-0.208,-0.000,-2.079", Rotation="-14.576,0.000,0.000", Scale="1,1,1") {
    rotationMode=BMM_Additive
    rotationSpace=BCS_BoneSpace
    translationMode=BMM_Additive
    translationSpace=BCS_BoneSpace
    alphaBoolBlend {blendOption=Linear}
    alphaScaleBiasClamp {bInterpResult=true, interpSpeedDecreasing=7, interpSpeedIncreasing=7}
    boneToModify {boneName=dyn_skirt_side_01_r}
    rotation {pitch=-14.576}
    translation {x=-0.208, z=-2.079}
   }
   n43 = "Transform (Modify) Bone - Bone: dyn_skirt_frontMid_01_r"(ComponentPose=@n42, Alpha=@n41, Translation="-0.208,-0.000,-2.079", Rotation="-14.576,0.000,0.000", Scale="1,1,1") {
    rotationMode=BMM_Additive
    rotationSpace=BCS_BoneSpace
    translationMode=BMM_Additive
    translationSpace=BCS_BoneSpace
    alphaBoolBlend {blendOption=Linear}
    alphaScaleBiasClamp {bInterpResult=true, interpSpeedDecreasing=7, interpSpeedIncreasing=7}
    boneToModify {boneName=dyn_skirt_frontMid_01_r}
    rotation {pitch=-14.576}
    translation {x=-0.208, z=-2.079}
   }
   n44 = Get front_end_rear_alpha()
   n45 = "Transform (Modify) Bone - Bone: driver_skirt_rear_01_l"(ComponentPose=@n43, Alpha=@n44, Translation="0,-2.5,0", Rotation="0,-5,0", Scale="1,1,1") {
    rotationMode=BMM_Additive
    rotationSpace=BCS_BoneSpace
    translationMode=BMM_Additive
    translationSpace=BCS_BoneSpace
    alphaBoolBlend {blendOption=Linear}
    boneToModify {boneName=driver_skirt_rear_01_l}
    rotation {yaw=-5}
    translation {y=-2.5}
   }
   n46 = "Transform (Modify) Bone - Bone: driver_skirt_rear_01_r"(ComponentPose=@n45, Alpha=@n44, Translation="0,-2.5,0", Rotation="0,-5,0", Scale="1,1,1") {
    rotationMode=BMM_Additive
    rotationSpace=BCS_BoneSpace
    translationMode=BMM_Additive
    translationSpace=BCS_BoneSpace
    alphaBoolBlend {blendOption=Linear}
    boneToModify {boneName=driver_skirt_rear_01_r}
    rotation {yaw=-5}
    translation {y=-2.5}
   }
   Save cached pose 'AnimCache'(Pose=Component To Local(ComponentPose=@n46)) {cacheName=AnimCache}
   Get vector_force() -> (vector_force:Vector)
   Get gravity_override()
   Get vector_force()
  ]}
 EventGraph {nodes [
   Event BlueprintUpdateAnimation()
   > Get!(execute:Exec) -> (then:Exec, else:Exec, Skel_Mesh:Skeletal Mesh Component Object Reference)
   > Get!(execute) -> (then, else, Anim_Inst:Anim Instance Object Reference)
   > n1 = Sequence!(execute) -> (then_0, then_1, then_2:Exec, then_3:Exec, then_4:Exec, then_5:Exec, then_6:Exec)
   n2 = Gravity Override State Function DEPRECATED(
    execute=@n1.then_1
    CharacterPartAnimInstance=Self-Reference()
    GameTimeInSeconds=GetGameTimeInSeconds()
    GravityOverrideParams=Get GravityOverrideParams() -> (GravityOverrideParams:Gravity Override Params Struct Structure)
   )
   > Set gravity_override(gravity_override=@n2)
   n3 = Branch(
    execute=@n1.then_2
    Condition=Montage_IsPlaying(
     self=Get Anim_Inst() -> (Anim_Inst:Anim Instance Object Reference)
     Montage=/BRCosmetics/Animation/Game/MainPlayer/Emotes/ShaolinSipUp/Emote_ShaolinSipUp_CMF_M
    )
   )
   > Set emote_shaolin_sip_active(emote_shaolin_sip_active:double=1) -> (Output_Get:double)
   Set emote_shaolin_sip_active(execute=@n3.else, emote_shaolin_sip_active=0)
   Set thigh_up_alpha(
    execute=@n1.then_6
    thigh_up_alpha:double=GetAlphaFromReferenceTransformDelta(
     SkeletalMeshComponent=Get Skel_Mesh()
     BoneName=thigh_r
     BaseName=pelvis
     DeltaComponentType=QuaternionTwist
     InputRange=(X=0,Y=100)
     TwistAxis="0,0,1"
    )
   ) -> (Output_Get:double)
   n4 = Get PawnSpeed2D()
   n5 = Branch(
    execute=@n1.then_0
    Condition="float <= float"!(A:double=MapRangeClamped(Value=@n4, InRangeA=0, InRangeB=150), B:double=0) -> (ReturnValue:bool) {is=title K2Node_CallFunction}
   )
   > Set cloth_alpha(cloth_alpha:double=0.45) -> (Output_Get:double)
   > Set side_cloth_alpha(side_cloth_alpha:double=0.7) -> (Output_Get:double)
   > n6 = Branch(Condition=Get wearing_long_cape() -> (wearing_long_cape:bool))
   > n7 = Set cloth_alpha(cloth_alpha=0.25)
   n8 = Branch(execute=[@n7.then, @n6.else], Condition=Get bIsCrouching())
   > Set feather_alpha(feather_alpha:double=1) -> (Output_Get:double)
   > Set crouch_skirt_alpha(crouch_skirt_alpha:double=1) -> (Output_Get:double)
   > Set side_cloth_alpha(side_cloth_alpha=0.6)
   Set feather_alpha(execute=@n8.else, feather_alpha=0)
   > Set crouch_skirt_alpha(crouch_skirt_alpha=0)
   > n9 = Branch(Condition=Get emote_active() -> (emote_active:bool))
   > n10 = Set cloth_alpha(cloth_alpha=1)
   Branch(execute=[@n9.else, @n10.then], Condition=Get bCanPlayCustomAnimations() -> (bCanPlayCustomAnimations:bool))
   > Set cloth_alpha(cloth_alpha=1)
   Set cloth_alpha(execute=@n5.else, cloth_alpha=1)
   > Set feather_alpha(feather_alpha=1)
   > Set side_cloth_alpha(side_cloth_alpha=0.7)
   > n11 = Branch(Condition=Get bIsCrouching())
   > n12 = Set crouch_skirt_alpha(crouch_skirt_alpha=0.1)
   Branch(execute=[@n13.then, @n12.then], Condition=Get wearing_long_cape())
   > Set cloth_alpha(cloth_alpha=0.3)
   n13 = Set crouch_skirt_alpha(execute=@n11.else, crouch_skirt_alpha=0)
   Event BlueprintInitializeAnimation()
   > Set Skel_Mesh(Skel_Mesh=GetOwnerSkeletalMeshComponent())
   > n14 = Get!(execute) -> (then, else, Skel_Mesh:Skeletal Mesh Component Object Reference)
   > Set Anim_Inst(Anim_Inst:Anim Instance Object Reference=GetAnimInstance(self=@n14)) -> (Output_Get:Anim Instance Object Reference)
   Event OnBeginPlayingEmote() {is=title K2Node_Event}
   > Set emote_active(emote_active:bool=true) -> (Output_Get:bool)
   Event OnEndPlayingEmote() {is=title K2Node_Event}
   > Set emote_active(emote_active=false)
   Event OnCharacterCustomizationLoaded()
   n15 = CheckForBackbling()
   > Set wearing_backbling(wearing_backbling:bool=@n15) -> (Output_Get:bool)
   > Set wearing_long_cape(wearing_long_cape:bool=@n15.bIsBackblingLong) -> (Output_Get:bool)
   n16 = Event OnNotifyPartsStopped!() -> (OutputDelegate, then, Context:Name=None) {is=title K2Node_Event}
   > Branch(Condition="Equal (Name)"!(A:Name=@n16.Context, B:Name=Female_Commando_Idle_02) -> (ReturnValue:bool) {is=title K2Node_CallFunction})
   > Set front_end_rear_alpha(front_end_rear_alpha:double=0) -> (Output_Get:double)
   n17 = Event OnNotifyPartsStarted!() -> (OutputDelegate, then, Context:Name=None) {is=title K2Node_Event}
   > Branch(Condition="Equal (Name)"(A=@n17.Context, B=Female_Commando_Idle_02))
   > Set front_end_rear_alpha(front_end_rear_alpha=1)
   n18 = Branch(Condition=Get bIsSkyDiving())
   > n19 = Branch(Condition=Get bIsParachuteOpened())
   > Set gravity_override(gravity_override="0, 0,-800")
   n20 = Break Fort Anim Input Skydiving External Force(FortAnimInput_SkydivingExternalForce=Get SkydivingExternalForceInput()) -> (
    bUseSkydivingVectorForce=~
    SkydivingVectorForce=~
    HeadToPelvisDirection=~
    FloatingMultiplier=~
    FloatingAdditive=~
    DivingMultiplier=~
    DivingAdditive=~
    ParachutingMultiplier=~
    ParachutingAdditive=~
   )
   > Set vector_force(vector_force:Vector=@n20.SkydivingVectorForce) -> (Output_Get:Vector)
   n21 = Branch(execute=@n19.else, Condition="float > float"(A=Get PawnVelocityZ(), B=50) {is=title K2Node_CallFunction})
   > Set gravity_override(gravity_override="0, 0,-300")
   > Set vector_force(vector_force=@n20.SkydivingVectorForce)
   Set gravity_override(execute=@n21.else, gravity_override="0, 0,-300")
   > Set vector_force(vector_force=@n20.SkydivingVectorForce)
   > Branch(Condition=Get bCanPlayCustomAnimations())
   > Set gravity_override(gravity_override="0, 0,150")
   Set gravity_override(execute=@n18.else, gravity_override="0, 0,-920")
   > Set vector_force(vector_force="0, 0, 0")
  ]}
]
variables [
 Anim_Inst {type=AnimInstance, default=null}
 cloth_alpha {type=number}
 crouch_skirt_alpha {type=number}
 emote_active {type=boolean}
 emote_shaolin_sip_active {type=number}
 feather_alpha {type=number}
 front_end_rear_alpha {type=number}
 gravity_override {type=Vector}
 GravityOverrideParams {type=GravityOverrideParamsStruct}
 side_cloth_alpha {type=number, default=1}
 Skel_Mesh {type=SkeletalMeshComponent}
 thigh_up_alpha {type=number}
 vector_force {type=Vector}
 wearing_backbling {type=boolean}
 wearing_long_cape {type=boolean}
]
